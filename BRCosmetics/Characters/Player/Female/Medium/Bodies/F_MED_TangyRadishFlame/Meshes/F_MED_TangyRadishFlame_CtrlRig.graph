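# Control Rig for the F_MED_TangyRadishFlame female body cosmetic, correcting body and clothing deformation.
# Forward solve runs shared DeformRig, then weighted parent constraints seat skirt, shirt, ribbon and lace base bones on body bones.
# skirtSetup offsets skirt bones using thigh-up alphas read from root anim attributes.
# HeadUp reshapes neck_02 via a spherical pose reader on head, or at fixed full weight when playingEmote is true.
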ControlRigBlueprint /BRCosmetics/Characters/Player/Female/Medium/Bodies/F_MED_TangyRadishFlame/Meshes/F_MED_TangyRadishFlame_CtrlRig
graphs [
 New Function_ContainedGraph {nodes [
   n1 = RigVMFunctionEntryNode()
   RigVMFunction_Sequence = &RigVMFunction_Sequence(ExecuteContext=@n1.ExecuteContext)
   &RigUnit_ParentConstraint(
    AdvancedSettings=(RotationOrderForFilter=XZY)
    Child=(Type=Bone,Name="base_ribbon_leg_l")
    Parents=((Item=(Type=Bone,Name="pelvis"),Weight=0.0003),(Item=(Type=Bone,Name="deform_knee_out_l"),Weight=0.0155),(Item=(Type=Bone,Name="thigh_twist_01_l"),Weight=0.1048),(Item=(Type=Bone,Name="thigh_l"),Weight=0.8794))
    Weight=1
    ExecutePin=@RigVMFunction_Sequence.A
   )
   > &RigUnit_ParentConstraint(
    AdvancedSettings=(RotationOrderForFilter=XZY)
    Child=(Type=Bone,Name="base_laces_foot_l")
    Parents=((Item=(Type=Bone,Name="calf_twist_01_l"),Weight=0.0002),(Item=(Type=Bone,Name="ball_l"),Weight=0.2172),(Item=(Type=Bone,Name="foot_l"),Weight=0.7826))
    Weight=1
   )
   > &RigUnit_ParentConstraint(
    AdvancedSettings=(RotationOrderForFilter=XZY)
    Child=(Type=Bone,Name="base_front_skirt_l")
    Parents="((Item=(Type=Bone,Name=\"thigh_twist_01_r\"),Weight=0.0005),(Item=(Type=Bone,Name=\"pelvis\"),Weight=0.142),(Item=(Type=Bone,Name=\"thigh_l\"),Weight=0.1752),(Item=(Type=Bone,Name=\"deform_groin_l\"),Weight=0.1855),(Item=(Type=Bone,Name=\"thigh_twist_01_l\"),Weight= ...(truncated)"
    Weight=1
   )
   > &RigUnit_ParentConstraint(
    AdvancedSettings=(RotationOrderForFilter=XZY)
    Child=(Type=Bone,Name="base_side_skirt_l")
    Parents="((Item=(Type=Bone,Name=\"deform_groin_l\"),Weight=0.0151),(Item=(Type=Bone,Name=\"deform_glute_l\"),Weight=0.0339),(Item=(Type=Bone,Name=\"thigh_l\"),Weight=0.1334),(Item=(Type=Bone,Name=\"pelvis\"),Weight=0.1466),(Item=(Type=Bone,Name=\"thigh_twist_01_l\"),Weight=0 ...(truncated)"
    Weight=1
   )
   > &RigUnit_ParentConstraint(
    AdvancedSettings=(RotationOrderForFilter=XZY)
    Child=(Type=Bone,Name="base_bk_skirt_l")
    Parents=((Item=(Type=Bone,Name="thigh_l"),Weight=0.0001),(Item=(Type=Bone,Name="thigh_twist_01_l"),Weight=0.1735),(Item=(Type=Bone,Name="pelvis"),Weight=0.3926),(Item=(Type=Bone,Name="deform_glute_l"),Weight=0.4337))
    Weight=1
   )
   > &RigUnit_ParentConstraint(
    AdvancedSettings=(RotationOrderForFilter=XZY)
    Child=(Type=Bone,Name="base_ribbon_chest_B")
    Parents=((Item=(Type=Bone,Name="spine_02"),Weight=0.036),(Item=(Type=Bone,Name="spine_03"),Weight=0.0362),(Item=(Type=Bone,Name="spine_05"),Weight=0.0489),(Item=(Type=Bone,Name="spine_04"),Weight=0.8789))
    Weight=1
   )
   > &RigUnit_ParentConstraint(
    AdvancedSettings=(RotationOrderForFilter=XZY)
    Child=(Type=Bone,Name="base_ribbon_chest_A")
    Parents=((Item=(Type=Bone,Name="spine_05"),Weight=0.2448),(Item=(Type=Bone,Name="spine_04"),Weight=0.7552))
    Weight=1
   )
   > &RigUnit_ParentConstraint(
    AdvancedSettings=(RotationOrderForFilter=XZY)
    Child=(Type=Bone,Name="base_side_shirt_l")
    Parents="((Item=(Type=Bone,Name=\"upperarm_twist_01_l\"),Weight=0.0118),(Item=(Type=Bone,Name=\"pelvis\"),Weight=0.0298),(Item=(Type=Bone,Name=\"clavicle_l\"),Weight=0.0302),(Item=(Type=Bone,Name=\"spine_05\"),Weight=0.0344),(Item=(Type=Bone,Name=\"spine_01\"),Weight=0.0458) ...(truncated)"
    Weight=1
   )
   > &RigUnit_ParentConstraint(
    AdvancedSettings=(RotationOrderForFilter=XZY)
    Child=(Type=Bone,Name="base_bk_shirt_l")
    Parents="((Item=(Type=Bone,Name=\"upperarm_twist_01_l\"),Weight=0.0033),(Item=(Type=Bone,Name=\"clavicle_l\"),Weight=0.0419),(Item=(Type=Bone,Name=\"spine_05\"),Weight=0.0433),(Item=(Type=Bone,Name=\"pelvis\"),Weight=0.0486),(Item=(Type=Bone,Name=\"spine_01\"),Weight=0.0798) ...(truncated)"
    Weight=1
   )
   > &RigUnit_ParentConstraint(
    AdvancedSettings=(RotationOrderForFilter=XZY)
    Child=(Type=Bone,Name="base_front_shirt_r")
    Parents=((Item=(Type=Bone,Name="pelvis"),Weight=0.0108),(Item=(Type=Bone,Name="spine_01"),Weight=0.0237),(Item=(Type=Bone,Name="spine_03"),Weight=0.1066),(Item=(Type=Bone,Name="spine_02"),Weight=0.1544),(Item=(Type=Bone,Name="spine_04"),Weight=0.7044))
    Weight=1
   )
   > &RigUnit_ParentConstraint(
    AdvancedSettings=(RotationOrderForFilter=XZY)
    Child=(Type=Bone,Name="base_front_shirt_l")
    Parents=((Item=(Type=Bone,Name="pelvis"),Weight=0.0108),(Item=(Type=Bone,Name="spine_01"),Weight=0.0237),(Item=(Type=Bone,Name="spine_03"),Weight=0.1066),(Item=(Type=Bone,Name="spine_02"),Weight=0.1544),(Item=(Type=Bone,Name="spine_04"),Weight=0.7044))
    Weight=1
   )
   > &RigUnit_ParentConstraint(
    AdvancedSettings=(RotationOrderForFilter=XZY)
    Child=(Type=Bone,Name="base_bk_shirt_r")
    Parents="((Item=(Type=Bone,Name=\"upperarm_twist_01_r\"),Weight=0.0033),(Item=(Type=Bone,Name=\"clavicle_r\"),Weight=0.0419),(Item=(Type=Bone,Name=\"spine_05\"),Weight=0.0433),(Item=(Type=Bone,Name=\"pelvis\"),Weight=0.0486),(Item=(Type=Bone,Name=\"spine_01\"),Weight=0.0798) ...(truncated)"
    Weight=1
   )
   > &RigUnit_ParentConstraint(
    AdvancedSettings=(RotationOrderForFilter=XZY)
    Child=(Type=Bone,Name="base_side_shirt_r")
    Parents="((Item=(Type=Bone,Name=\"upperarm_twist_01_r\"),Weight=0.0161),(Item=(Type=Bone,Name=\"pelvis\"),Weight=0.0241),(Item=(Type=Bone,Name=\"clavicle_r\"),Weight=0.0354),(Item=(Type=Bone,Name=\"spine_05\"),Weight=0.0404),(Item=(Type=Bone,Name=\"spine_01\"),Weight=0.0423) ...(truncated)"
    Weight=1
   )
   > &RigUnit_ParentConstraint(
    AdvancedSettings=(RotationOrderForFilter=XZY)
    Child=(Type=Bone,Name="base_ribbon_arm_r")
    Parents=((Item=(Type=Bone,Name="upperarm_twist_02_r"),Weight=0.1638),(Item=(Type=Bone,Name="upperarm_r"),Weight=0.1646),(Item=(Type=Bone,Name="upperarm_twist_01_r"),Weight=0.6716))
    Weight=1
   )
   > &RigUnit_ParentConstraint(
    AdvancedSettings=(RotationOrderForFilter=XZY)
    Child=(Type=Bone,Name="base_laces_foot_r")
    Parents=((Item=(Type=Bone,Name="calf_twist_01_r"),Weight=0.0002),(Item=(Type=Bone,Name="ball_r"),Weight=0.2172),(Item=(Type=Bone,Name="foot_r"),Weight=0.7826))
    Weight=1
   )
   > &RigUnit_ParentConstraint(
    AdvancedSettings=(RotationOrderForFilter=XZY)
    Child=(Type=Bone,Name="base_front_skirt_r")
    Parents="((Item=(Type=Bone,Name=\"thigh_twist_01_l\"),Weight=0.0005),(Item=(Type=Bone,Name=\"pelvis\"),Weight=0.1422),(Item=(Type=Bone,Name=\"thigh_r\"),Weight=0.1753),(Item=(Type=Bone,Name=\"deform_groin_r\"),Weight=0.1853),(Item=(Type=Bone,Name=\"thigh_twist_01_r\"),Weight ...(truncated)"
    Weight=1
   )
   > &RigUnit_ParentConstraint(
    AdvancedSettings=(RotationOrderForFilter=XZY)
    Child=(Type=Bone,Name="base_side_skirt_r")
    Parents="((Item=(Type=Bone,Name=\"deform_groin_r\"),Weight=0.0137),(Item=(Type=Bone,Name=\"deform_glute_r\"),Weight=0.0324),(Item=(Type=Bone,Name=\"thigh_r\"),Weight=0.1317),(Item=(Type=Bone,Name=\"pelvis\"),Weight=0.1445),(Item=(Type=Bone,Name=\"thigh_twist_01_r\"),Weight=0 ...(truncated)"
    Weight=1
   )
   > &RigUnit_ParentConstraint(
    AdvancedSettings=(RotationOrderForFilter=XZY)
    Child=(Type=Bone,Name="base_bk_skirt_r")
    Parents="((Item=(Type=Bone,Name=\"deform_groin_r\"),Weight=0.0002),(Item=(Type=Bone,Name=\"thigh_l\"),Weight=0.0003),(Item=(Type=Bone,Name=\"deform_glute_l\"),Weight=0.0045),(Item=(Type=Bone,Name=\"thigh_twist_01_l\"),Weight=0.0054),(Item=(Type=Bone,Name=\"thigh_r\"),Weight= ...(truncated)"
    Weight=1
   )
   > &RigUnit_ParentConstraint(
    AdvancedSettings=(RotationOrderForFilter=XZY)
    Child=(Type=Bone,Name="base_ribbon_pelvis")
    Parents="((Item=(Type=Bone,Name=\"thigh_twist_01_r\"),Weight=0.0001),(Item=(Type=Bone,Name=\"thigh_l\"),Weight=0.0003),(Item=(Type=Bone,Name=\"spine_02\"),Weight=0.0396),(Item=(Type=Bone,Name=\"thigh_twist_01_l\"),Weight=0.0892),(Item=(Type=Bone,Name=\"deform_groin_l\"),Weig ...(truncated)"
    Weight=1
   )
   RigVMFunctionReturnNode(ExecuteContext=@RigVMFunction_Sequence.B)
   &RigUnit_ParentConstraint(
    AdvancedSettings=(RotationOrderForFilter=XZY)
    Child=(Type=Bone,Name="base_ribbon_pelvis")
    ExecutePin=()
    Parents=((Item=(Type=Bone,Name="thigh_twist_01_l"),Weight=0.0221),(Item=(Type=Bone,Name="spine_02"),Weight=0.0608),(Item=(Type=Bone,Name="deform_groin_l"),Weight=0.067),(Item=(Type=Bone,Name="pelvis"),Weight=0.85))
    Weight=1
   )
  ]}
 New Function_ContainedGraph {nodes [
   n1 = RigVMFunctionEntryNode()
   RigVMFunction_Sequence = &RigVMFunction_Sequence(ExecuteContext=@n1.ExecuteContext)
   &RigUnit_ModifyTransforms(
    ItemToModify="((Item=(Type=Bone,Name=\"base_front_skirt_l\"),Transform=(Rotation=(X=0,Y=-0.2164,Z=0,W=0.9763),Translation=(X=0,Y=0,Z=0),Scale3D=(X=1,Y=1,Z=1))),(Item=(Type=Bone,Name=\"base_side_skirt_l\"),Transform=(Rotation=(X=0,Y=0,Z=0,W=1),Translation=(X=6,Y=0,Z=0),Scale ...(truncated)"
    WeightMinimum=0
    Weight=RigVMDispatchNode(BoneName=root, Default=0, Name=deformRig_l_thigh_up_alpha) {template="DISPATCH_RigDispatch_GetAnimAttribute(in Name,in BoneName,in Default,out Value,out Found)"}.Value
    ExecutePin=@RigVMFunction_Sequence.A
   )
   > &RigUnit_ModifyTransforms(
    ItemToModify="((Item=(Type=Bone,Name=\"base_front_skirt_r\"),Transform=(Rotation=(X=0,Y=-0.2164,Z=0,W=0.9763),Translation=(X=0,Y=0,Z=0),Scale3D=(X=1,Y=1,Z=1))),(Item=(Type=Bone,Name=\"base_side_skirt_r\"),Transform=(Rotation=(X=0,Y=0,Z=0,W=1),Translation=(X=0,Y=0,Z=-0.5),Sc ...(truncated)"
    WeightMinimum=0
    Weight=RigVMDispatchNode(BoneName=root, Default=0, Name=deformRig_r_thigh_up_alpha) {template="DISPATCH_RigDispatch_GetAnimAttribute(in Name,in BoneName,in Default,out Value,out Found)"}.Value
   )
   RigVMFunctionReturnNode(ExecuteContext=@RigVMFunction_Sequence.B)
  ]}
 New Function_ContainedGraph {nodes [
   n1 = RigVMFunctionEntryNode()
   RigVMFunction_Sequence = &RigVMFunction_Sequence(ExecuteContext=@n1.ExecuteContext)
   RigVMFunction_ControlFlowBranch = &RigVMFunction_ControlFlowBranch(ExecuteContext=@RigVMFunction_Sequence.A, Condition=RigVMVariableNode() {variable=playingEmote}.Value)
   n2 = &RigUnit_SphericalPoseReader(
    ActiveRegionScaleFactors=(PositiveWidth=1,NegativeWidth=0.2,PositiveHeight=1.5,NegativeHeight=1.5)
    ActiveRegionSize=0.1
    Debug=(bDrawDebug=false,bDraw2D=false,bDrawLocalAxes=false,DebugScale=25,DebugSegments=20,DebugThickness=0.25)
    DriverAxis=(X=0,Y=1,Z=0)
    DriverItem=(Type=Bone,Name="head")
    FalloffRegionScaleFactors=(PositiveWidth=3,NegativeWidth=1.8,PositiveHeight=2.5,NegativeHeight=2.5)
    FalloffSize=0.2
    FlipHeightScaling=False
    FlipWidthScaling=False
    OptionalParentItem=(Type=Bone,Name="None")
    RotationOffset=(X=90,Y=0,Z=-55)
    ExecutePin=@RigVMFunction_ControlFlowBranch.False
   )
   Multiply = &RigVMFunction_MathDoubleMul(B=2, A=@n2.OutputParam)
   > &RigUnit_ModifyTransforms(
    ItemToModify=((Item=(Type=Bone,Name="neck_02"),Transform=(Rotation=(X=0,Y=0,Z=0,W=1),Translation=(X=0,Y=0,Z=0),Scale3D=(X=1,Y=1.5,Z=1.4))))
    WeightMinimum=0
    Weight=@Multiply.Result
   )
   &RigUnit_ModifyTransforms(
    ItemToModify=((Item=(Type=Bone,Name="neck_02"),Transform=(Rotation=(X=0,Y=0,Z=0,W=1),Translation=(X=0,Y=1,Z=0),Scale3D=(X=1,Y=1.4,Z=1.4))))
    Weight=1
    WeightMinimum=0
    ExecutePin=@RigVMFunction_ControlFlowBranch.True
   )
   RigVMFunctionReturnNode(ExecuteContext=@RigVMFunction_Sequence.B)
  ]}
 RigVMFunctionLibrary {nodes [
   Constraints = RigVMCollapseNode(ExecuteContext=())
   HeadUp = RigVMCollapseNode(ExecuteContext=())
   skirtSetup = RigVMCollapseNode(ExecuteContext=())
  ]}
 RigVMModel {nodes [
   n1 = &RigUnit_PrepareForExecution()
   &RigUnit_HierarchyImportFromSkeleton(NameSpace=None, bIncludeCurves=False, bIncludeMeshSockets=False, bIncludeVirtualBones=True, ExecutePin=@n1.ExecutePin)
   RigUnit_BeginExecution = &RigUnit_BeginExecution()
   FN_DeformRig_FNC_1 = RigVMFunctionReferenceNode(
    DeformRig_ Curve="(deformRig_l_upperarm_up_alpha_72_#1=False,deformRig_r_upperarm_up_alpha_73_#2=False,deformRig_l_upperarm_fr_alpha_74_#3=False,deformRig_r_upperarm_fr_alpha_75_CB51B8 ...(truncated)"
    DeformRig_Bool="(upperarm_1_#4=True,elbow_3_#5=True,wrist_7_#6=True,groin_8_#7=True,glute_17_#8=True,knee_9_40C408324CA261 ...(truncated)"
    DeformRig_Debug=(upperarm_debug_18_#4=False,wrist_debug_20_#6=False,lat_debug_26_#9=False,groinGlute_debug_27_#7=False)
    DeformRig_Value="(upperarm_up_18_#10=(Rotation=(X=0,Y=0,Z=0,W=1),Translation=(X=0,Y=0,Z=8),Scale3D=(X=0.7,Y=1,Z=1)),upperarm_fwd_fr_19_#11=(Rotation=(X=0,Y=0,Z=0,W=1),Translation=(X=0,Y=10,Z=0),Scale3D=(X=0.5,Y=1,Z= ...(truncated)"
    ExecuteContext=@RigUnit_BeginExecution.ExecutePin
   )
   > Constraints = RigVMFunctionReferenceNode()
   > skirtSetup = RigVMFunctionReferenceNode()
   > HeadUp = RigVMFunctionReferenceNode()
   addFKCtrl = RigVMFunctionReferenceNode(
    Custom_FKCtrl=()
    ExecuteContext=()
    Options_FKCtrl="(pelvis_1_#12=true,spine_3_#13=true,clavicle_5_#14=true,arm_7_#15=true,leg_9_#16=true,head_11_46DC15EF465CD2979 ...(truncated)"
    Shape_Settings="(pelvis_2_#17=(bVisible=True,Name=\"Hexagon_Thin\",Color=(R=0.0117,G=0.8281,B=1,A=1),Transform=(Rotation=(X=0,Y=-0.7071,Z=0,W=0.7071),Translation=(X=2,Y=0,Z=0),Scale3D=(X=5,Y=5,Z=7))),spine_4_#18=(bVi ...(truncated)"
   )
   setFKCtrl = RigVMFunctionReferenceNode(ExecuteContext=(), setFK_Ctrl=false)
  ]}
]
variables [playingEmote {type=bool, default=False}]
