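# Animation Blueprint for the F_MED_TangyRadishFlame female cosmetic body, adding secondary motion to ribbons, ribbon tapes, shoelaces and shirt flaps.
# Serial chains of single-bone Anim Dynamics nodes simulate root-relative with reduced gravity, angular limits and springs, mostly stiff for bows and laces, softer for tapes.
# Two visible chains start from the 'skirtDyn_cache' cached pose and repeat bones with different tuning, while a separate pelvis chain has no visible input.
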AnimBlueprint /BRCosmetics/Characters/Player/Female/Medium/Bodies/F_MED_TangyRadishFlame/Meshes/F_MED_TangyRadishFlame_AnimBP
target_skeleton=&/F_MED_TangyRadishFlame_Skeleton
graphs [
 AnimGraph {nodes [
   n1 = Anim Dynamics - Bone: dyn_ribbon_pelvis_l!(ComponentPose, bAlphaBoolEnabled:bool=True, ExternalForce="0,0,0") -> (Pose) {
    alphaInputType=Bool
    angularDampingOverride=1
    angularSpringConstant=10000
    bAngularSpring=true
    bOverrideAngularDamping=true
    bUseGravityOverride=true
    gravityScale=0.35
    simulationSpace=RootRelative
    alphaBoolBlend {blendInTime=0.2, blendOption=Linear, blendOutTime=0.2}
    boundBone {boneName=dyn_ribbon_pelvis_l}
    componentLinearAccScale=<1 1 1>
    componentLinearVelScale=<1 1 1>
    gravityOverride {z=-980}
    physicsBodyDefinitions [{
      boundBone {boneName=dyn_ribbon_pelvis_l}
      boxExtents=<15 15 15>
      constraintSetup {angularLimitsMax=<10 0 10>, angularLimitsMin=<-10 -20 -10>, angularTarget=<1 0 0>}
      localJointOffset=<3 0 0>
     }]
   }
   n2 = Anim Dynamics - Bone: dyn_ribbon_pelvis_r!(ComponentPose=@n1, bAlphaBoolEnabled:bool=True, ExternalForce="0,0,0") -> (Pose) {
    alphaInputType=Bool
    angularDampingOverride=1
    angularSpringConstant=10000
    bAngularSpring=true
    bOverrideAngularDamping=true
    bUseGravityOverride=true
    gravityScale=0.35
    simulationSpace=RootRelative
    alphaBoolBlend {blendInTime=0.2, blendOption=Linear, blendOutTime=0.2}
    boundBone {boneName=dyn_ribbon_pelvis_r}
    componentLinearAccScale=<1 1 1>
    componentLinearVelScale=<1 1 1>
    gravityOverride {z=-980}
    physicsBodyDefinitions [{
      boundBone {boneName=dyn_ribbon_pelvis_r}
      boxExtents=<15 15 15>
      constraintSetup {angularLimitsMax=<10 0 10>, angularLimitsMin=<-10 -20 -10>, angularTarget=<1 0 0>}
      localJointOffset=<3 0 0>
     }]
   }
   n3 = Anim Dynamics - Bone: dyn_ribbon_tape_pelvis_l_1!(ComponentPose=@n2, bAlphaBoolEnabled:bool=True, ExternalForce="0,0,0") -> (Pose) {
    alphaInputType=Bool
    angularDampingOverride=0.9999
    angularSpringConstant=500
    bAngularSpring=true
    bOverrideAngularDamping=true
    bUseGravityOverride=true
    gravityScale=0.35
    simulationSpace=RootRelative
    alphaBoolBlend {blendInTime=0.2, blendOption=Linear, blendOutTime=0.2}
    boundBone {boneName=dyn_ribbon_tape_pelvis_l_1}
    componentAppliedLinearAccClamp=<3000 3000 3000>
    componentLinearAccScale=<0.5 0.5 0.5>
    componentLinearVelScale=<0.5 0.5 0.5>
    gravityOverride {z=-980}
    physicsBodyDefinitions [{
      boundBone {boneName=dyn_ribbon_tape_pelvis_l_1}
      boxExtents=<15 10 10>
      constraintSetup {angularLimitsMax=<5 -25 30>, angularLimitsMin=<-5 -55 -30>, angularTarget=<1 0 0>}
      localJointOffset=<3 0 0>
     }]
   }
   n4 = Anim Dynamics - Bone: dyn_ribbon_tape_pelvis_r_1!(ComponentPose=@n3, bAlphaBoolEnabled:bool=True, ExternalForce="0,0,0") -> (Pose) {
    alphaInputType=Bool
    angularDampingOverride=0.9999
    angularSpringConstant=500
    bAngularSpring=true
    bOverrideAngularDamping=true
    bUseGravityOverride=true
    gravityScale=0.35
    simulationSpace=RootRelative
    alphaBoolBlend {blendInTime=0.2, blendOption=Linear, blendOutTime=0.2}
    boundBone {boneName=dyn_ribbon_tape_pelvis_r_1}
    componentAppliedLinearAccClamp=<3000 3000 3000>
    componentLinearAccScale=<0.5 0.5 0.5>
    componentLinearVelScale=<0.5 0.5 0.5>
    gravityOverride {z=-980}
    physicsBodyDefinitions [{
      boundBone {boneName=dyn_ribbon_tape_pelvis_r_1}
      boxExtents=<15 10 10>
      constraintSetup {angularLimitsMax=<5 -25 30>, angularLimitsMin=<-5 -55 -30>, angularTarget=<1 0 0>}
      localJointOffset=<3 0 0>
     }]
   }
   n5 = Anim Dynamics - Bone: dyn_ribbon_tape_pelvis_l_2!(ComponentPose=@n4, bAlphaBoolEnabled:bool=True, ExternalForce="0,0,0") -> (Pose) {
    alphaInputType=Bool
    angularDampingOverride=0.9999
    angularSpringConstant=500
    bAngularSpring=true
    bOverrideAngularDamping=true
    bUseGravityOverride=true
    gravityScale=0.35
    simulationSpace=RootRelative
    alphaBoolBlend {blendInTime=0.2, blendOption=Linear, blendOutTime=0.2}
    boundBone {boneName=dyn_ribbon_tape_pelvis_l_2}
    componentAppliedLinearAccClamp=<3000 3000 3000>
    componentLinearAccScale=<0.5 0.5 0.5>
    componentLinearVelScale=<0.5 0.5 0.5>
    gravityOverride {z=-980}
    physicsBodyDefinitions [{
      boundBone {boneName=dyn_ribbon_tape_pelvis_l_2}
      boxExtents=<15 10 10>
      constraintSetup {angularLimitsMax=<5 0 30>, angularLimitsMin=<-5 -45 -30>, angularTarget=<1 0 0>}
      localJointOffset=<3 0 0>
     }]
   }
   Anim Dynamics - Bone: dyn_ribbon_tape_pelvis_r_2!(ComponentPose=@n5, bAlphaBoolEnabled:bool=True, ExternalForce="0,0,0") -> (Pose) {
    alphaInputType=Bool
    angularDampingOverride=0.9999
    angularSpringConstant=500
    bAngularSpring=true
    bOverrideAngularDamping=true
    bUseGravityOverride=true
    gravityScale=0.35
    simulationSpace=RootRelative
    alphaBoolBlend {blendInTime=0.2, blendOption=Linear, blendOutTime=0.2}
    boundBone {boneName=dyn_ribbon_tape_pelvis_r_2}
    componentAppliedLinearAccClamp=<3000 3000 3000>
    componentLinearAccScale=<0.5 0.5 0.5>
    componentLinearVelScale=<0.5 0.5 0.5>
    gravityOverride {z=-980}
    physicsBodyDefinitions [{
      boundBone {boneName=dyn_ribbon_tape_pelvis_r_2}
      boxExtents=<15 10 10>
      constraintSetup {angularLimitsMax=<5 0 30>, angularLimitsMin=<-5 -45 -30>, angularTarget=<1 0 0>}
      localJointOffset=<3 0 0>
     }]
   }
   n6 = Anim Dynamics - Bone: dyn_ribbon_leg_l_r!(ComponentPose=Local To Component(LocalPose=Use cached pose 'skirtDyn_cache'()), bAlphaBoolEnabled:bool=True, ExternalForce="0,0,0") -> (Pose) {
    alphaInputType=Bool
    angularDampingOverride=1
    angularSpringConstant=10000
    bAngularSpring=true
    bOverrideAngularDamping=true
    bUseGravityOverride=true
    gravityScale=0.35
    simulationSpace=RootRelative
    alphaBoolBlend {blendInTime=0.2, blendOption=Linear, blendOutTime=0.2}
    boundBone {boneName=dyn_ribbon_leg_l_r}
    componentLinearAccScale=<1 1 1>
    componentLinearVelScale=<1 1 1>
    gravityOverride {z=-980}
    physicsBodyDefinitions [{
      boundBone {boneName=dyn_ribbon_leg_l_r}
      boxExtents=<20 20 20>
      constraintSetup {angularLimitsMax=<10 7.5 10>, angularLimitsMin=<-10 -7.5 -10>, angularTarget=<1 0 0>}
      localJointOffset=<3 0 0>
     }]
   }
   n7 = Anim Dynamics - Bone: dyn_ribbon_leg_l_l!(ComponentPose=@n6, bAlphaBoolEnabled:bool=True, ExternalForce="0,0,0") -> (Pose) {
    alphaInputType=Bool
    angularDampingOverride=1
    angularSpringConstant=10000
    bAngularSpring=true
    bOverrideAngularDamping=true
    bUseGravityOverride=true
    gravityScale=0.35
    simulationSpace=RootRelative
    alphaBoolBlend {blendInTime=0.2, blendOption=Linear, blendOutTime=0.2}
    boundBone {boneName=dyn_ribbon_leg_l_l}
    componentLinearAccScale=<1 1 1>
    componentLinearVelScale=<1 1 1>
    gravityOverride {z=-980}
    physicsBodyDefinitions [{
      boundBone {boneName=dyn_ribbon_leg_l_l}
      boxExtents=<20 20 20>
      constraintSetup {angularLimitsMax=<10 7.5 10>, angularLimitsMin=<-10 -7.5 -10>, angularTarget=<1 0 0>}
      localJointOffset=<3 0 0>
     }]
   }
   n8 = Anim Dynamics - Bone: dyn_ribbon_tape_leg_l!(ComponentPose=@n7, bAlphaBoolEnabled:bool=True, ExternalForce="0,0,0") -> (Pose) {
    alphaInputType=Bool
    angularDampingOverride=0.9999
    angularSpringConstant=500
    bAngularSpring=true
    bOverrideAngularDamping=true
    bUseGravityOverride=true
    gravityScale=0.35
    simulationSpace=RootRelative
    alphaBoolBlend {blendInTime=0.2, blendOption=Linear, blendOutTime=0.2}
    boundBone {boneName=dyn_ribbon_tape_leg_l}
    componentAppliedLinearAccClamp=<3000 3000 3000>
    componentLinearAccScale=<0.5 0.5 0.5>
    componentLinearVelScale=<0.5 0.5 0.5>
    gravityOverride {z=-980}
    physicsBodyDefinitions [{
      boundBone {boneName=dyn_ribbon_tape_leg_l}
      boxExtents=<20 15 15>
      constraintSetup {angularLimitsMax=<5 0 30>, angularLimitsMin=<-5 -45 -30>, angularTarget=<1 0 0>}
      localJointOffset=<3 0 0>
     }]
   }
   n9 = Anim Dynamics - Bone: dyn_laces_foot_l_r!(ComponentPose=@n8, bAlphaBoolEnabled:bool=True, ExternalForce="0,0,0") -> (Pose) {
    alphaInputType=Bool
    angularDampingOverride=1
    angularSpringConstant=10000
    bAngularSpring=true
    bOverrideAngularDamping=true
    bUseGravityOverride=true
    gravityScale=0.35
    simulationSpace=RootRelative
    alphaBoolBlend {blendInTime=0.2, blendOption=Linear, blendOutTime=0.2}
    boundBone {boneName=dyn_laces_foot_l_r}
    componentLinearAccScale=<1 1 1>
    componentLinearVelScale=<1 1 1>
    gravityOverride {z=-980}
    physicsBodyDefinitions [{
      boundBone {boneName=dyn_laces_foot_l_r}
      boxExtents=<20 20 20>
      constraintSetup {angularLimitsMax=<10 0 10>, angularLimitsMin=<-10 -10 -10>, angularTarget=<1 0 0>}
      localJointOffset=<3 0 0>
     }]
   }
   n10 = Anim Dynamics - Bone: dyn_laces_foot_l_l!(ComponentPose=@n9, bAlphaBoolEnabled:bool=True, ExternalForce="0,0,0") -> (Pose) {
    alphaInputType=Bool
    angularDampingOverride=1
    angularSpringConstant=10000
    bAngularSpring=true
    bOverrideAngularDamping=true
    bUseGravityOverride=true
    gravityScale=0.35
    simulationSpace=RootRelative
    alphaBoolBlend {blendInTime=0.2, blendOption=Linear, blendOutTime=0.2}
    boundBone {boneName=dyn_laces_foot_l_l}
    componentLinearAccScale=<1 1 1>
    componentLinearVelScale=<1 1 1>
    gravityOverride {z=-980}
    physicsBodyDefinitions [{
      boundBone {boneName=dyn_laces_foot_l_l}
      boxExtents=<20 20 20>
      constraintSetup {angularLimitsMax=<10 0 10>, angularLimitsMin=<-10 -10 -10>, angularTarget=<1 0 0>}
      localJointOffset=<3 0 0>
     }]
   }
   n11 = Anim Dynamics - Bone: dyn_laces_foot_r_r!(ComponentPose=@n10, bAlphaBoolEnabled:bool=True, ExternalForce="0,0,0") -> (Pose) {
    alphaInputType=Bool
    angularDampingOverride=1
    angularSpringConstant=10000
    bAngularSpring=true
    bOverrideAngularDamping=true
    bUseGravityOverride=true
    gravityScale=0.35
    simulationSpace=RootRelative
    alphaBoolBlend {blendInTime=0.2, blendOption=Linear, blendOutTime=0.2}
    boundBone {boneName=dyn_laces_foot_r_r}
    componentLinearAccScale=<1 1 1>
    componentLinearVelScale=<1 1 1>
    gravityOverride {z=-980}
    physicsBodyDefinitions [{
      boundBone {boneName=dyn_laces_foot_r_r}
      boxExtents=<20 20 20>
      constraintSetup {angularLimitsMax=<10 0 10>, angularLimitsMin=<-10 -10 -10>, angularTarget=<1 0 0>}
      localJointOffset=<3 0 0>
     }]
   }
   n12 = Anim Dynamics - Bone: dyn_laces_foot_r_l!(ComponentPose=@n11, bAlphaBoolEnabled:bool=True, ExternalForce="0,0,0") -> (Pose) {
    alphaInputType=Bool
    angularDampingOverride=1
    angularSpringConstant=10000
    bAngularSpring=true
    bOverrideAngularDamping=true
    bUseGravityOverride=true
    gravityScale=0.35
    simulationSpace=RootRelative
    alphaBoolBlend {blendInTime=0.2, blendOption=Linear, blendOutTime=0.2}
    boundBone {boneName=dyn_laces_foot_r_l}
    componentLinearAccScale=<1 1 1>
    componentLinearVelScale=<1 1 1>
    gravityOverride {z=-980}
    physicsBodyDefinitions [{
      boundBone {boneName=dyn_laces_foot_r_l}
      boxExtents=<20 20 20>
      constraintSetup {angularLimitsMax=<10 0 10>, angularLimitsMin=<-10 -10 -10>, angularTarget=<1 0 0>}
      localJointOffset=<3 0 0>
     }]
   }
   n13 = Anim Dynamics - Bone: dyn_ribbon_chest_a_r!(ComponentPose=@n12, bAlphaBoolEnabled:bool=True, ExternalForce="0,0,0") -> (Pose) {
    alphaInputType=Bool
    angularDampingOverride=1
    angularSpringConstant=10000
    bAngularSpring=true
    bOverrideAngularDamping=true
    bUseGravityOverride=true
    gravityScale=0.35
    simulationSpace=RootRelative
    alphaBoolBlend {blendInTime=0.2, blendOption=Linear, blendOutTime=0.2}
    boundBone {boneName=dyn_ribbon_chest_a_r}
    componentLinearAccScale=<1 1 1>
    componentLinearVelScale=<1 1 1>
    gravityOverride {z=-980}
    physicsBodyDefinitions [{
      boundBone {boneName=dyn_ribbon_chest_a_r}
      boxExtents=<20 20 20>
      constraintSetup {angularLimitsMax=<10 10 10>, angularLimitsMin=<-10 -10 -10>, angularTarget=<1 0 0>}
      localJointOffset=<3 0 0>
     }]
   }
   n14 = Anim Dynamics - Bone: dyn_ribbon_chest_a_l!(ComponentPose=@n13, bAlphaBoolEnabled:bool=True, ExternalForce="0,0,0") -> (Pose) {
    alphaInputType=Bool
    angularDampingOverride=1
    angularSpringConstant=10000
    bAngularSpring=true
    bOverrideAngularDamping=true
    bUseGravityOverride=true
    gravityScale=0.35
    simulationSpace=RootRelative
    alphaBoolBlend {blendInTime=0.2, blendOption=Linear, blendOutTime=0.2}
    boundBone {boneName=dyn_ribbon_chest_a_l}
    componentLinearAccScale=<1 1 1>
    componentLinearVelScale=<1 1 1>
    gravityOverride {z=-980}
    physicsBodyDefinitions [{
      boundBone {boneName=dyn_ribbon_chest_a_l}
      boxExtents=<20 20 20>
      constraintSetup {angularLimitsMax=<10 10 10>, angularLimitsMin=<-10 -10 -10>, angularTarget=<1 0 0>}
      localJointOffset=<3 0 0>
     }]
   }
   n15 = Anim Dynamics - Bone: dyn_ribbon_chest_tape_A_r!(ComponentPose=@n14, bAlphaBoolEnabled:bool=True, ExternalForce="0,0,0") -> (Pose) {
    alphaInputType=Bool
    angularDampingOverride=0.9999
    angularSpringConstant=500
    bAngularSpring=true
    bOverrideAngularDamping=true
    bUseGravityOverride=true
    gravityScale=0.35
    simulationSpace=RootRelative
    alphaBoolBlend {blendInTime=0.2, blendOption=Linear, blendOutTime=0.2}
    boundBone {boneName=dyn_ribbon_chest_tape_A_r}
    componentAppliedLinearAccClamp=<3000 3000 3000>
    componentLinearAccScale=<0.5 0.5 0.5>
    componentLinearVelScale=<0.5 0.5 0.5>
    gravityOverride {z=-980}
    physicsBodyDefinitions [{
      boundBone {boneName=dyn_ribbon_chest_tape_A_r}
      boxExtents=<20 15 15>
      constraintSetup {angularLimitsMax=<5 0 30>, angularLimitsMin=<-5 -45 -30>, angularTarget=<1 0 0>}
      localJointOffset=<3 0 0>
     }]
   }
   n16 = Anim Dynamics - Bone: dyn_ribbon_chest_tape_A_l!(ComponentPose=@n15, bAlphaBoolEnabled:bool=True, ExternalForce="0,0,0") -> (Pose) {
    alphaInputType=Bool
    angularDampingOverride=0.9999
    angularSpringConstant=500
    bAngularSpring=true
    bOverrideAngularDamping=true
    bUseGravityOverride=true
    gravityScale=0.35
    simulationSpace=RootRelative
    alphaBoolBlend {blendInTime=0.2, blendOption=Linear, blendOutTime=0.2}
    boundBone {boneName=dyn_ribbon_chest_tape_A_l}
    componentAppliedLinearAccClamp=<3000 3000 3000>
    componentLinearAccScale=<0.5 0.5 0.5>
    componentLinearVelScale=<0.5 0.5 0.5>
    gravityOverride {z=-980}
    physicsBodyDefinitions [{
      boundBone {boneName=dyn_ribbon_chest_tape_A_l}
      boxExtents=<20 15 15>
      constraintSetup {angularLimitsMax=<5 0 30>, angularLimitsMin=<-5 -45 -30>, angularTarget=<1 0 0>}
      localJointOffset=<3 0 0>
     }]
   }
   n17 = Anim Dynamics - Bone: dyn_ribbon_chest_B_r!(ComponentPose=@n16, bAlphaBoolEnabled:bool=True, ExternalForce="0,0,0") -> (Pose) {
    alphaInputType=Bool
    angularDampingOverride=1
    angularSpringConstant=10000
    bAngularSpring=true
    bOverrideAngularDamping=true
    bUseGravityOverride=true
    gravityScale=0.35
    simulationSpace=RootRelative
    alphaBoolBlend {blendInTime=0.2, blendOption=Linear, blendOutTime=0.2}
    boundBone {boneName=dyn_ribbon_chest_B_r}
    componentLinearAccScale=<1 1 1>
    componentLinearVelScale=<1 1 1>
    gravityOverride {z=-980}
    physicsBodyDefinitions [{
      boundBone {boneName=dyn_ribbon_chest_B_r}
      boxExtents=<20 20 20>
      constraintSetup {angularLimitsMax=<10 10 10>, angularLimitsMin=<-10 -10 -10>, angularTarget=<1 0 0>}
      localJointOffset=<3 0 0>
     }]
   }
   n18 = Anim Dynamics - Bone: dyn_ribbon_chest_B_l!(ComponentPose=@n17, bAlphaBoolEnabled:bool=True, ExternalForce="0,0,0") -> (Pose) {
    alphaInputType=Bool
    angularDampingOverride=1
    angularSpringConstant=10000
    bAngularSpring=true
    bOverrideAngularDamping=true
    bUseGravityOverride=true
    gravityScale=0.35
    simulationSpace=RootRelative
    alphaBoolBlend {blendInTime=0.2, blendOption=Linear, blendOutTime=0.2}
    boundBone {boneName=dyn_ribbon_chest_B_l}
    componentLinearAccScale=<1 1 1>
    componentLinearVelScale=<1 1 1>
    gravityOverride {z=-980}
    physicsBodyDefinitions [{
      boundBone {boneName=dyn_ribbon_chest_B_l}
      boxExtents=<20 20 20>
      constraintSetup {angularLimitsMax=<10 10 10>, angularLimitsMin=<-10 -10 -10>, angularTarget=<1 0 0>}
      localJointOffset=<3 0 0>
     }]
   }
   n19 = Anim Dynamics - Bone: dyn_ribbon_tape_chest_B_r_1!(ComponentPose=@n18, bAlphaBoolEnabled:bool=True, ExternalForce="0,0,0") -> (Pose) {
    alphaInputType=Bool
    angularDampingOverride=0.9999
    angularSpringConstant=500
    bAngularSpring=true
    bOverrideAngularDamping=true
    bUseGravityOverride=true
    gravityScale=0.35
    simulationSpace=RootRelative
    alphaBoolBlend {blendInTime=0.2, blendOption=Linear, blendOutTime=0.2}
    boundBone {boneName=dyn_ribbon_tape_chest_B_r_1}
    componentAppliedLinearAccClamp=<3000 3000 3000>
    componentLinearAccScale=<0.5 0.5 0.5>
    componentLinearVelScale=<0.5 0.5 0.5>
    gravityOverride {z=-980}
    physicsBodyDefinitions [{
      boundBone {boneName=dyn_ribbon_tape_chest_B_r_1}
      boxExtents=<20 15 15>
      constraintSetup {angularLimitsMax=<5 0 30>, angularLimitsMin=<-5 -45 -30>, angularTarget=<1 0 0>}
      localJointOffset=<3 0 0>
     }]
   }
   n20 = Anim Dynamics - Bone: dyn_ribbon_tape_chest_B_l_1!(ComponentPose=@n19, bAlphaBoolEnabled:bool=True, ExternalForce="0,0,0") -> (Pose) {
    alphaInputType=Bool
    angularDampingOverride=0.9999
    angularSpringConstant=500
    bAngularSpring=true
    bOverrideAngularDamping=true
    bUseGravityOverride=true
    gravityScale=0.35
    simulationSpace=RootRelative
    alphaBoolBlend {blendInTime=0.2, blendOption=Linear, blendOutTime=0.2}
    boundBone {boneName=dyn_ribbon_tape_chest_B_l_1}
    componentAppliedLinearAccClamp=<3000 3000 3000>
    componentLinearAccScale=<0.5 0.5 0.5>
    componentLinearVelScale=<0.5 0.5 0.5>
    gravityOverride {z=-980}
    physicsBodyDefinitions [{
      boundBone {boneName=dyn_ribbon_tape_chest_B_l_1}
      boxExtents=<20 15 15>
      constraintSetup {angularLimitsMax=<5 0 30>, angularLimitsMin=<-5 -45 -30>, angularTarget=<1 0 0>}
      localJointOffset=<3 0 0>
     }]
   }
   n21 = Anim Dynamics - Bone: dyn_ribbon_tape_chest_B_r_2!(ComponentPose=@n20, bAlphaBoolEnabled:bool=True, ExternalForce="0,0,0") -> (Pose) {
    alphaInputType=Bool
    angularDampingOverride=0.9999
    angularSpringConstant=500
    bAngularSpring=true
    bOverrideAngularDamping=true
    bUseGravityOverride=true
    gravityScale=0.35
    simulationSpace=RootRelative
    alphaBoolBlend {blendInTime=0.2, blendOption=Linear, blendOutTime=0.2}
    boundBone {boneName=dyn_ribbon_tape_chest_B_r_2}
    componentAppliedLinearAccClamp=<3000 3000 3000>
    componentLinearAccScale=<0.5 0.5 0.5>
    componentLinearVelScale=<0.5 0.5 0.5>
    gravityOverride {z=-980}
    physicsBodyDefinitions [{
      boundBone {boneName=dyn_ribbon_tape_chest_B_r_2}
      boxExtents=<20 15 15>
      constraintSetup {angularLimitsMax=<5 0 30>, angularLimitsMin=<-5 -45 -30>, angularTarget=<1 0 0>}
      localJointOffset=<3 0 0>
     }]
   }
   n22 = Anim Dynamics - Bone: dyn_ribbon_tape_chest_B_l_2!(ComponentPose=@n21, bAlphaBoolEnabled:bool=True, ExternalForce="0,0,0") -> (Pose) {
    alphaInputType=Bool
    angularDampingOverride=0.9999
    angularSpringConstant=500
    bAngularSpring=true
    bOverrideAngularDamping=true
    bUseGravityOverride=true
    gravityScale=0.35
    simulationSpace=RootRelative
    alphaBoolBlend {blendInTime=0.2, blendOption=Linear, blendOutTime=0.2}
    boundBone {boneName=dyn_ribbon_tape_chest_B_l_2}
    componentAppliedLinearAccClamp=<3000 3000 3000>
    componentLinearAccScale=<0.5 0.5 0.5>
    componentLinearVelScale=<0.5 0.5 0.5>
    gravityOverride {z=-980}
    physicsBodyDefinitions [{
      boundBone {boneName=dyn_ribbon_tape_chest_B_l_2}
      boxExtents=<20 15 15>
      constraintSetup {angularLimitsMax=<5 0 30>, angularLimitsMin=<-5 -45 -30>, angularTarget=<1 0 0>}
      localJointOffset=<3 0 0>
     }]
   }
   n23 = Anim Dynamics - Bone: dyn_front_shirt_r!(ComponentPose=@n22, bAlphaBoolEnabled:bool=True, ExternalForce="0,0,0") -> (Pose) {
    alphaInputType=Bool
    angularDampingOverride=1
    angularSpringConstant=10000
    bAngularSpring=true
    bOverrideAngularDamping=true
    bUseGravityOverride=true
    gravityScale=0.35
    simulationSpace=RootRelative
    alphaBoolBlend {blendInTime=0.2, blendOption=Linear, blendOutTime=0.2}
    boundBone {boneName=dyn_front_shirt_r}
    componentLinearAccScale=<1 1 1>
    componentLinearVelScale=<1 1 1>
    gravityOverride {z=-980}
    physicsBodyDefinitions [{
      boundBone {boneName=dyn_front_shirt_r}
      boxExtents=<13 13 13>
      constraintSetup {angularLimitsMax=<5 15 10>, angularLimitsMin=<-5 -10 -10>, angularTarget=<1 0 0>}
      localJointOffset=<3 0 0>
     }]
   }
   n24 = Anim Dynamics - Bone: dyn_front_shirt_l!(ComponentPose=@n23, bAlphaBoolEnabled:bool=True, ExternalForce="0,0,0") -> (Pose) {
    alphaInputType=Bool
    angularDampingOverride=1
    angularSpringConstant=10000
    bAngularSpring=true
    bOverrideAngularDamping=true
    bUseGravityOverride=true
    gravityScale=0.35
    simulationSpace=RootRelative
    alphaBoolBlend {blendInTime=0.2, blendOption=Linear, blendOutTime=0.2}
    boundBone {boneName=dyn_front_shirt_l}
    componentLinearAccScale=<1 1 1>
    componentLinearVelScale=<1 1 1>
    gravityOverride {z=-980}
    physicsBodyDefinitions [{
      boundBone {boneName=dyn_front_shirt_l}
      boxExtents=<13 13 13>
      constraintSetup {angularLimitsMax=<5 15 10>, angularLimitsMin=<-5 -10 -10>, angularTarget=<1 0 0>}
      localJointOffset=<3 0 0>
     }]
   }
   n25 = Anim Dynamics - Bone: dyn_bk_shirt_r!(ComponentPose=@n24, bAlphaBoolEnabled:bool=True, ExternalForce="0,0,0") -> (Pose) {
    alphaInputType=Bool
    angularDampingOverride=0.9999
    angularSpringConstant=500
    bAngularSpring=true
    bOverrideAngularDamping=true
    bUseGravityOverride=true
    gravityScale=0.35
    simulationSpace=RootRelative
    alphaBoolBlend {blendInTime=0.2, blendOption=Linear, blendOutTime=0.2}
    boundBone {boneName=dyn_bk_shirt_r}
    componentAppliedLinearAccClamp=<3000 3000 3000>
    componentLinearAccScale=<0.5 0.5 0.5>
    componentLinearVelScale=<0.5 0.5 0.5>
    gravityOverride {z=-980}
    physicsBodyDefinitions [{
      boundBone {boneName=dyn_bk_shirt_r}
      boxExtents=<13 13 13>
      constraintSetup {angularLimitsMax=<5 5 10>, angularLimitsMin=<-5 -40 -10>, angularTarget=<1 0 0>}
      localJointOffset=<3 0 0>
     }]
   }
   n26 = Anim Dynamics - Bone: dyn_bk_shirt_l!(ComponentPose=@n25, bAlphaBoolEnabled:bool=True, ExternalForce="0,0,0") -> (Pose) {
    alphaInputType=Bool
    angularDampingOverride=0.9999
    angularSpringConstant=500
    bAngularSpring=true
    bOverrideAngularDamping=true
    bUseGravityOverride=true
    gravityScale=0.35
    simulationSpace=RootRelative
    alphaBoolBlend {blendInTime=0.2, blendOption=Linear, blendOutTime=0.2}
    boundBone {boneName=dyn_bk_shirt_l}
    componentAppliedLinearAccClamp=<3000 3000 3000>
    componentLinearAccScale=<0.5 0.5 0.5>
    componentLinearVelScale=<0.5 0.5 0.5>
    gravityOverride {z=-980}
    physicsBodyDefinitions [{
      boundBone {boneName=dyn_bk_shirt_l}
      boxExtents=<13 13 13>
      constraintSetup {angularLimitsMax=<5 5 10>, angularLimitsMin=<-5 -40 -10>, angularTarget=<1 0 0>}
      localJointOffset=<3 0 0>
     }]
   }
   n27 = Anim Dynamics - Bone: dyn_bk_shirt_c!(ComponentPose=@n26, bAlphaBoolEnabled:bool=True, ExternalForce="0,0,0") -> (Pose) {
    alphaInputType=Bool
    angularDampingOverride=0.9999
    angularSpringConstant=500
    bAngularSpring=true
    bOverrideAngularDamping=true
    bUseGravityOverride=true
    gravityScale=0.35
    simulationSpace=RootRelative
    alphaBoolBlend {blendInTime=0.2, blendOption=Linear, blendOutTime=0.2}
    boundBone {boneName=dyn_bk_shirt_c}
    componentAppliedLinearAccClamp=<3000 3000 3000>
    componentLinearAccScale=<0.5 0.5 0.5>
    componentLinearVelScale=<0.5 0.5 0.5>
    gravityOverride {z=-980}
    physicsBodyDefinitions [{
      boundBone {boneName=dyn_bk_shirt_c}
      boxExtents=<13 13 13>
      constraintSetup {angularLimitsMax=<5 0 10>, angularLimitsMin=<-5 -15 -10>, angularTarget=<1 0 0>}
      localJointOffset=<3 0 0>
     }]
   }
   n28 = Anim Dynamics - Bone: dyn_side_shirt_r!(ComponentPose=@n27, bAlphaBoolEnabled:bool=True, ExternalForce="0,0,0") -> (Pose) {
    alphaInputType=Bool
    angularDampingOverride=0.9999
    angularSpringConstant=500
    bAngularSpring=true
    bOverrideAngularDamping=true
    bUseGravityOverride=true
    gravityScale=0.35
    simulationSpace=RootRelative
    alphaBoolBlend {blendInTime=0.2, blendOption=Linear, blendOutTime=0.2}
    boundBone {boneName=dyn_side_shirt_r}
    componentAppliedLinearAccClamp=<3000 3000 3000>
    componentLinearAccScale=<0.5 0.5 0.5>
    componentLinearVelScale=<0.5 0.5 0.5>
    gravityOverride {z=-980}
    physicsBodyDefinitions [{
      boundBone {boneName=dyn_side_shirt_r}
      boxExtents=<13 13 13>
      constraintSetup {angularLimitsMax=<5 15 10>, angularLimitsMin=<-5 -40 -10>, angularTarget=<1 0 0>}
      localJointOffset=<3 0 0>
     }]
   }
   n29 = Anim Dynamics - Bone: dyn_side_shirt_l!(ComponentPose=@n28, bAlphaBoolEnabled:bool=True, ExternalForce="0,0,0") -> (Pose) {
    alphaInputType=Bool
    angularDampingOverride=0.9999
    angularSpringConstant=500
    bAngularSpring=true
    bOverrideAngularDamping=true
    bUseGravityOverride=true
    gravityScale=0.35
    simulationSpace=RootRelative
    alphaBoolBlend {blendInTime=0.2, blendOption=Linear, blendOutTime=0.2}
    boundBone {boneName=dyn_side_shirt_l}
    componentAppliedLinearAccClamp=<3000 3000 3000>
    componentLinearAccScale=<0.5 0.5 0.5>
    componentLinearVelScale=<0.5 0.5 0.5>
    gravityOverride {z=-980}
    physicsBodyDefinitions [{
      boundBone {boneName=dyn_side_shirt_l}
      boxExtents=<13 13 13>
      constraintSetup {angularLimitsMax=<5 15 10>, angularLimitsMin=<-5 -40 -10>, angularTarget=<1 0 0>}
      localJointOffset=<3 0 0>
     }]
   }
   n30 = Anim Dynamics - Bone: dyn_ribbon_arm_r_r!(ComponentPose=@n29, bAlphaBoolEnabled:bool=True, ExternalForce="0,0,0") -> (Pose) {
    alphaInputType=Bool
    angularDampingOverride=1
    angularSpringConstant=10000
    bAngularSpring=true
    bOverrideAngularDamping=true
    bUseGravityOverride=true
    gravityScale=0.35
    simulationSpace=RootRelative
    alphaBoolBlend {blendInTime=0.2, blendOption=Linear, blendOutTime=0.2}
    boundBone {boneName=dyn_ribbon_arm_r_r}
    componentLinearAccScale=<1 1 1>
    componentLinearVelScale=<1 1 1>
    gravityOverride {z=-980}
    physicsBodyDefinitions [{
      boundBone {boneName=dyn_ribbon_arm_r_r}
      boxExtents=<20 15 15>
      constraintSetup {angularLimitsMax=<10 7.5 10>, angularLimitsMin=<-10 7.5 -10>, angularTarget=<1 0 0>}
      localJointOffset=<3 0 0>
     }]
   }
   n31 = Anim Dynamics - Bone: dyn_ribbon_arm_r_l!(ComponentPose=@n30, bAlphaBoolEnabled:bool=True, ExternalForce="0,0,0") -> (Pose) {
    alphaInputType=Bool
    angularDampingOverride=1
    angularSpringConstant=10000
    bAngularSpring=true
    bOverrideAngularDamping=true
    bUseGravityOverride=true
    gravityScale=0.35
    simulationSpace=RootRelative
    alphaBoolBlend {blendInTime=0.2, blendOption=Linear, blendOutTime=0.2}
    boundBone {boneName=dyn_ribbon_arm_r_l}
    componentLinearAccScale=<1 1 1>
    componentLinearVelScale=<1 1 1>
    gravityOverride {z=-980}
    physicsBodyDefinitions [{
      boundBone {boneName=dyn_ribbon_arm_r_l}
      boxExtents=<20 15 15>
      constraintSetup {angularLimitsMax=<10 7.5 10>, angularLimitsMin=<-10 -7.5 -10>, angularTarget=<1 0 0>}
      localJointOffset=<3 0 0>
     }]
   }
   n32 = Anim Dynamics - Bone: dyn_ribbon_tape_arm_r!(ComponentPose=@n31, bAlphaBoolEnabled:bool=True, ExternalForce="0,0,0") -> (Pose) {
    alphaInputType=Bool
    angularDampingOverride=0.9999
    angularSpringConstant=500
    bAngularSpring=true
    bOverrideAngularDamping=true
    bUseGravityOverride=true
    gravityScale=0.35
    simulationSpace=RootRelative
    alphaBoolBlend {blendInTime=0.2, blendOption=Linear, blendOutTime=0.2}
    boundBone {boneName=dyn_ribbon_tape_arm_r}
    componentAppliedLinearAccClamp=<3000 3000 3000>
    componentLinearAccScale=<0.5 0.5 0.5>
    componentLinearVelScale=<0.5 0.5 0.5>
    gravityOverride {z=-980}
    physicsBodyDefinitions [{
      boundBone {boneName=dyn_ribbon_tape_arm_r}
      boxExtents=<20 15 15>
      constraintSetup {angularLimitsMax=<5 0 30>, angularLimitsMin=<-5 -45 -30>, angularTarget=<1 0 0>}
      localJointOffset=<3 0 0>
     }]
   }
   n33 = Anim Dynamics - Bone: dyn_ribbon_leg_l_r(ComponentPose=Local To Component(LocalPose=Use cached pose 'skirtDyn_cache'()), bAlphaBoolEnabled=True, ExternalForce="0,0,0") {
    alphaInputType=Bool
    angularDampingOverride=1
    angularSpringConstant=10000
    bAngularSpring=true
    bOverrideAngularDamping=true
    bUseGravityOverride=true
    gravityScale=0.35
    simulationSpace=RootRelative
    alphaBoolBlend {blendInTime=0.2, blendOption=Linear, blendOutTime=0.2}
    boundBone {boneName=dyn_ribbon_leg_l_r}
    componentLinearAccScale=<1 1 1>
    componentLinearVelScale=<1 1 1>
    gravityOverride {z=-980}
    physicsBodyDefinitions [{
      boundBone {boneName=dyn_ribbon_leg_l_r}
      boxExtents=<15 15 15>
      constraintSetup {angularLimitsMax=<10 7.5 10>, angularLimitsMin=<-10 -7.5 -10>, angularTarget=<1 0 0>}
      localJointOffset=<3 0 0>
     }]
    simSpaceSettings {maxAngularAcceleration=35, maxAngularVelocity=7, simSpaceAngularAlpha=0.7}
   }
   n34 = Anim Dynamics - Bone: dyn_ribbon_leg_l_l(ComponentPose=@n33, bAlphaBoolEnabled=True, ExternalForce="0,0,0") {
    alphaInputType=Bool
    angularDampingOverride=1
    angularSpringConstant=10000
    bAngularSpring=true
    bOverrideAngularDamping=true
    bUseGravityOverride=true
    gravityScale=0.35
    simulationSpace=RootRelative
    alphaBoolBlend {blendInTime=0.2, blendOption=Linear, blendOutTime=0.2}
    boundBone {boneName=dyn_ribbon_leg_l_l}
    componentLinearAccScale=<1 1 1>
    componentLinearVelScale=<1 1 1>
    gravityOverride {z=-980}
    physicsBodyDefinitions [{
      boundBone {boneName=dyn_ribbon_leg_l_l}
      boxExtents=<15 15 15>
      constraintSetup {angularLimitsMax=<10 7.5 10>, angularLimitsMin=<-10 -7.5 -10>, angularTarget=<1 0 0>}
      localJointOffset=<3 0 0>
     }]
    simSpaceSettings {maxAngularAcceleration=35, maxAngularVelocity=7, simSpaceAngularAlpha=0.7}
   }
   n35 = Anim Dynamics - Bone: dyn_ribbon_tape_leg_l(ComponentPose=@n34, bAlphaBoolEnabled=True, ExternalForce="0,0,0") {
    alphaInputType=Bool
    angularDampingOverride=0.9999
    angularSpringConstant=500
    bAngularSpring=true
    bOverrideAngularDamping=true
    bUseGravityOverride=true
    gravityScale=0.35
    simulationSpace=RootRelative
    alphaBoolBlend {blendInTime=0.2, blendOption=Linear, blendOutTime=0.2}
    boundBone {boneName=dyn_ribbon_tape_leg_l}
    componentLinearAccScale=<1 1 1>
    componentLinearVelScale=<1 1 1>
    gravityOverride {z=-980}
    physicsBodyDefinitions [{
      boundBone {boneName=dyn_ribbon_tape_leg_l}
      boxExtents=<15 10 10>
      constraintSetup {angularLimitsMax=<5 0 30>, angularLimitsMin=<-5 -45 -30>, angularTarget=<1 0 0>}
      localJointOffset=<3 0 0>
     }]
    simSpaceSettings {maxAngularAcceleration=35, maxAngularVelocity=7, simSpaceAngularAlpha=0.7}
   }
   n36 = Anim Dynamics - Bone: dyn_laces_foot_l_r(ComponentPose=@n35, bAlphaBoolEnabled=True, ExternalForce="0,0,0") {
    alphaInputType=Bool
    angularDampingOverride=1
    angularSpringConstant=10000
    bAngularSpring=true
    bOverrideAngularDamping=true
    bUseGravityOverride=true
    gravityScale=0.35
    simulationSpace=RootRelative
    alphaBoolBlend {blendInTime=0.2, blendOption=Linear, blendOutTime=0.2}
    boundBone {boneName=dyn_laces_foot_l_r}
    componentLinearAccScale=<1 1 1>
    componentLinearVelScale=<1 1 1>
    gravityOverride {z=-980}
    physicsBodyDefinitions [{
      boundBone {boneName=dyn_laces_foot_l_r}
      boxExtents=<15 15 15>
      constraintSetup {angularLimitsMax=<10 0 10>, angularLimitsMin=<-10 -10 -10>, angularTarget=<1 0 0>}
      localJointOffset=<3 0 0>
     }]
    simSpaceSettings {maxAngularAcceleration=35, maxAngularVelocity=7, simSpaceAngularAlpha=0.7}
   }
   n37 = Anim Dynamics - Bone: dyn_laces_foot_l_l(ComponentPose=@n36, bAlphaBoolEnabled=True, ExternalForce="0,0,0") {
    alphaInputType=Bool
    angularDampingOverride=1
    angularSpringConstant=10000
    bAngularSpring=true
    bOverrideAngularDamping=true
    bUseGravityOverride=true
    gravityScale=0.35
    simulationSpace=RootRelative
    alphaBoolBlend {blendInTime=0.2, blendOption=Linear, blendOutTime=0.2}
    boundBone {boneName=dyn_laces_foot_l_l}
    componentLinearAccScale=<1 1 1>
    componentLinearVelScale=<1 1 1>
    gravityOverride {z=-980}
    physicsBodyDefinitions [{
      boundBone {boneName=dyn_laces_foot_l_l}
      boxExtents=<15 15 15>
      constraintSetup {angularLimitsMax=<10 0 10>, angularLimitsMin=<-10 -10 -10>, angularTarget=<1 0 0>}
      localJointOffset=<3 0 0>
     }]
    simSpaceSettings {maxAngularAcceleration=35, maxAngularVelocity=7, simSpaceAngularAlpha=0.7}
   }
   n38 = Anim Dynamics - Bone: dyn_laces_foot_r_r(ComponentPose=@n37, bAlphaBoolEnabled=True, ExternalForce="0,0,0") {
    alphaInputType=Bool
    angularDampingOverride=1
    angularSpringConstant=10000
    bAngularSpring=true
    bOverrideAngularDamping=true
    bUseGravityOverride=true
    gravityScale=0.35
    simulationSpace=RootRelative
    alphaBoolBlend {blendInTime=0.2, blendOption=Linear, blendOutTime=0.2}
    boundBone {boneName=dyn_laces_foot_r_r}
    componentLinearAccScale=<1 1 1>
    componentLinearVelScale=<1 1 1>
    gravityOverride {z=-980}
    physicsBodyDefinitions [{
      boundBone {boneName=dyn_laces_foot_r_r}
      boxExtents=<15 15 15>
      constraintSetup {angularLimitsMax=<10 0 10>, angularLimitsMin=<-10 -10 -10>, angularTarget=<1 0 0>}
      localJointOffset=<3 0 0>
     }]
    simSpaceSettings {maxAngularAcceleration=35, maxAngularVelocity=7, simSpaceAngularAlpha=0.7}
   }
   n39 = Anim Dynamics - Bone: dyn_laces_foot_r_l(ComponentPose=@n38, bAlphaBoolEnabled=True, ExternalForce="0,0,0") {
    alphaInputType=Bool
    angularDampingOverride=1
    angularSpringConstant=10000
    bAngularSpring=true
    bOverrideAngularDamping=true
    bUseGravityOverride=true
    gravityScale=0.35
    simulationSpace=RootRelative
    alphaBoolBlend {blendInTime=0.2, blendOption=Linear, blendOutTime=0.2}
    boundBone {boneName=dyn_laces_foot_r_l}
    componentLinearAccScale=<1 1 1>
    componentLinearVelScale=<1 1 1>
    gravityOverride {z=-980}
    physicsBodyDefinitions [{
      boundBone {boneName=dyn_laces_foot_r_l}
      boxExtents=<15 15 15>
      constraintSetup {angularLimitsMax=<10 0 10>, angularLimitsMin=<-10 -10 -10>, angularTarget=<1 0 0>}
      localJointOffset=<3 0 0>
     }]
    simSpaceSettings {maxAngularAcceleration=35, maxAngularVelocity=7, simSpaceAngularAlpha=0.7}
   }
   n40 = Anim Dynamics - Bone: dyn_ribbon_chest_a_r(ComponentPose=@n39, bAlphaBoolEnabled=True, ExternalForce="0,0,0") {
    alphaInputType=Bool
    angularDampingOverride=1
    angularSpringConstant=10000
    bAngularSpring=true
    bOverrideAngularDamping=true
    bUseGravityOverride=true
    gravityScale=0.35
    simulationSpace=RootRelative
    alphaBoolBlend {blendInTime=0.2, blendOption=Linear, blendOutTime=0.2}
    boundBone {boneName=dyn_ribbon_chest_a_r}
    componentLinearAccScale=<1 1 1>
    componentLinearVelScale=<1 1 1>
    gravityOverride {z=-980}
    physicsBodyDefinitions [{
      boundBone {boneName=dyn_ribbon_chest_a_r}
      boxExtents=<15 15 15>
      constraintSetup {angularLimitsMax=<10 10 10>, angularLimitsMin=<-10 -10 -10>, angularTarget=<1 0 0>}
      localJointOffset=<3 0 0>
     }]
    simSpaceSettings {maxAngularAcceleration=35, maxAngularVelocity=7, simSpaceAngularAlpha=0.7}
   }
   n41 = Anim Dynamics - Bone: dyn_ribbon_chest_a_l(ComponentPose=@n40, bAlphaBoolEnabled=True, ExternalForce="0,0,0") {
    alphaInputType=Bool
    angularDampingOverride=1
    angularSpringConstant=10000
    bAngularSpring=true
    bOverrideAngularDamping=true
    bUseGravityOverride=true
    gravityScale=0.35
    simulationSpace=RootRelative
    alphaBoolBlend {blendInTime=0.2, blendOption=Linear, blendOutTime=0.2}
    boundBone {boneName=dyn_ribbon_chest_a_l}
    componentLinearAccScale=<1 1 1>
    componentLinearVelScale=<1 1 1>
    gravityOverride {z=-980}
    physicsBodyDefinitions [{
      boundBone {boneName=dyn_ribbon_chest_a_l}
      boxExtents=<15 15 15>
      constraintSetup {angularLimitsMax=<10 10 10>, angularLimitsMin=<-10 -10 -10>, angularTarget=<1 0 0>}
      localJointOffset=<3 0 0>
     }]
    simSpaceSettings {maxAngularAcceleration=35, maxAngularVelocity=7, simSpaceAngularAlpha=0.7}
   }
   n42 = Anim Dynamics - Bone: dyn_ribbon_chest_tape_A_r(ComponentPose=@n41, bAlphaBoolEnabled=True, ExternalForce="0,0,0") {
    alphaInputType=Bool
    angularDampingOverride=0.9999
    angularSpringConstant=500
    bAngularSpring=true
    bOverrideAngularDamping=true
    bUseGravityOverride=true
    gravityScale=0.35
    simulationSpace=RootRelative
    alphaBoolBlend {blendInTime=0.2, blendOption=Linear, blendOutTime=0.2}
    boundBone {boneName=dyn_ribbon_chest_tape_A_r}
    componentLinearAccScale=<1 1 1>
    componentLinearVelScale=<1 1 1>
    gravityOverride {z=-980}
    physicsBodyDefinitions [{
      boundBone {boneName=dyn_ribbon_chest_tape_A_r}
      boxExtents=<15 10 10>
      constraintSetup {angularLimitsMax=<5 0 30>, angularLimitsMin=<-5 -45 -30>, angularTarget=<1 0 0>}
      localJointOffset=<3 0 0>
     }]
    simSpaceSettings {maxAngularAcceleration=35, maxAngularVelocity=7, simSpaceAngularAlpha=0.7}
   }
   n43 = Anim Dynamics - Bone: dyn_ribbon_chest_tape_A_l(ComponentPose=@n42, bAlphaBoolEnabled=True, ExternalForce="0,0,0") {
    alphaInputType=Bool
    angularDampingOverride=0.9999
    angularSpringConstant=500
    bAngularSpring=true
    bOverrideAngularDamping=true
    bUseGravityOverride=true
    gravityScale=0.35
    simulationSpace=RootRelative
    alphaBoolBlend {blendInTime=0.2, blendOption=Linear, blendOutTime=0.2}
    boundBone {boneName=dyn_ribbon_chest_tape_A_l}
    componentLinearAccScale=<1 1 1>
    componentLinearVelScale=<1 1 1>
    gravityOverride {z=-980}
    physicsBodyDefinitions [{
      boundBone {boneName=dyn_ribbon_chest_tape_A_l}
      boxExtents=<15 10 10>
      constraintSetup {angularLimitsMax=<5 0 30>, angularLimitsMin=<-5 -45 -30>, angularTarget=<1 0 0>}
      localJointOffset=<3 0 0>
     }]
    simSpaceSettings {maxAngularAcceleration=35, maxAngularVelocity=7, simSpaceAngularAlpha=0.7}
   }
   n44 = Anim Dynamics - Bone: dyn_ribbon_chest_B_r(ComponentPose=@n43, bAlphaBoolEnabled=True, ExternalForce="0,0,0") {
    alphaInputType=Bool
    angularDampingOverride=1
    angularSpringConstant=10000
    bAngularSpring=true
    bOverrideAngularDamping=true
    bUseGravityOverride=true
    gravityScale=0.35
    simulationSpace=RootRelative
    alphaBoolBlend {blendInTime=0.2, blendOption=Linear, blendOutTime=0.2}
    boundBone {boneName=dyn_ribbon_chest_B_r}
    componentLinearAccScale=<1 1 1>
    componentLinearVelScale=<1 1 1>
    gravityOverride {z=-980}
    physicsBodyDefinitions [{
      boundBone {boneName=dyn_ribbon_chest_B_r}
      boxExtents=<15 15 15>
      constraintSetup {angularLimitsMax=<10 10 10>, angularLimitsMin=<-10 -10 -10>, angularTarget=<1 0 0>}
      localJointOffset=<3 0 0>
     }]
    simSpaceSettings {maxAngularAcceleration=35, maxAngularVelocity=7, simSpaceAngularAlpha=0.7}
   }
   n45 = Anim Dynamics - Bone: dyn_ribbon_chest_B_l(ComponentPose=@n44, bAlphaBoolEnabled=True, ExternalForce="0,0,0") {
    alphaInputType=Bool
    angularDampingOverride=1
    angularSpringConstant=10000
    bAngularSpring=true
    bOverrideAngularDamping=true
    bUseGravityOverride=true
    gravityScale=0.35
    simulationSpace=RootRelative
    alphaBoolBlend {blendInTime=0.2, blendOption=Linear, blendOutTime=0.2}
    boundBone {boneName=dyn_ribbon_chest_B_l}
    componentLinearAccScale=<1 1 1>
    componentLinearVelScale=<1 1 1>
    gravityOverride {z=-980}
    physicsBodyDefinitions [{
      boundBone {boneName=dyn_ribbon_chest_B_l}
      boxExtents=<15 15 15>
      constraintSetup {angularLimitsMax=<10 10 10>, angularLimitsMin=<-10 -10 -10>, angularTarget=<1 0 0>}
      localJointOffset=<3 0 0>
     }]
    simSpaceSettings {maxAngularAcceleration=35, maxAngularVelocity=7, simSpaceAngularAlpha=0.7}
   }
   n46 = Anim Dynamics - Bone: dyn_ribbon_tape_chest_B_r_1(ComponentPose=@n45, bAlphaBoolEnabled=True, ExternalForce="0,0,0") {
    alphaInputType=Bool
    angularDampingOverride=0.9999
    angularSpringConstant=500
    bAngularSpring=true
    bOverrideAngularDamping=true
    bUseGravityOverride=true
    gravityScale=0.35
    simulationSpace=RootRelative
    alphaBoolBlend {blendInTime=0.2, blendOption=Linear, blendOutTime=0.2}
    boundBone {boneName=dyn_ribbon_tape_chest_B_r_1}
    componentLinearAccScale=<1 1 1>
    componentLinearVelScale=<1 1 1>
    gravityOverride {z=-980}
    physicsBodyDefinitions [{
      boundBone {boneName=dyn_ribbon_tape_chest_B_r_1}
      boxExtents=<15 10 10>
      constraintSetup {angularLimitsMax=<5 0 30>, angularLimitsMin=<-5 -45 -30>, angularTarget=<1 0 0>}
      localJointOffset=<3 0 0>
     }]
    simSpaceSettings {maxAngularAcceleration=35, maxAngularVelocity=7, simSpaceAngularAlpha=0.7}
   }
   n47 = Anim Dynamics - Bone: dyn_ribbon_tape_chest_B_l_1(ComponentPose=@n46, bAlphaBoolEnabled=True, ExternalForce="0,0,0") {
    alphaInputType=Bool
    angularDampingOverride=0.9999
    angularSpringConstant=500
    bAngularSpring=true
    bOverrideAngularDamping=true
    bUseGravityOverride=true
    gravityScale=0.35
    simulationSpace=RootRelative
    alphaBoolBlend {blendInTime=0.2, blendOption=Linear, blendOutTime=0.2}
    boundBone {boneName=dyn_ribbon_tape_chest_B_l_1}
    componentLinearAccScale=<1 1 1>
    componentLinearVelScale=<1 1 1>
    gravityOverride {z=-980}
    physicsBodyDefinitions [{
      boundBone {boneName=dyn_ribbon_tape_chest_B_l_1}
      boxExtents=<15 10 10>
      constraintSetup {angularLimitsMax=<5 0 30>, angularLimitsMin=<-5 -45 -30>, angularTarget=<1 0 0>}
      localJointOffset=<3 0 0>
     }]
    simSpaceSettings {maxAngularAcceleration=35, maxAngularVelocity=7, simSpaceAngularAlpha=0.7}
   }
   n48 = Anim Dynamics - Bone: dyn_ribbon_tape_chest_B_r_2(ComponentPose=@n47, bAlphaBoolEnabled=True, ExternalForce="0,0,0") {
    alphaInputType=Bool
    angularDampingOverride=0.9999
    angularSpringConstant=500
    bAngularSpring=true
    bOverrideAngularDamping=true
    bUseGravityOverride=true
    gravityScale=0.35
    simulationSpace=RootRelative
    alphaBoolBlend {blendInTime=0.2, blendOption=Linear, blendOutTime=0.2}
    boundBone {boneName=dyn_ribbon_tape_chest_B_r_2}
    componentLinearAccScale=<1 1 1>
    componentLinearVelScale=<1 1 1>
    gravityOverride {z=-980}
    physicsBodyDefinitions [{
      boundBone {boneName=dyn_ribbon_tape_chest_B_r_2}
      boxExtents=<15 10 10>
      constraintSetup {angularLimitsMax=<5 0 30>, angularLimitsMin=<-5 -45 -30>, angularTarget=<1 0 0>}
      localJointOffset=<3 0 0>
     }]
    simSpaceSettings {maxAngularAcceleration=35, maxAngularVelocity=7, simSpaceAngularAlpha=0.7}
   }
   n49 = Anim Dynamics - Bone: dyn_ribbon_tape_chest_B_l_2(ComponentPose=@n48, bAlphaBoolEnabled=True, ExternalForce="0,0,0") {
    alphaInputType=Bool
    angularDampingOverride=0.9999
    angularSpringConstant=500
    bAngularSpring=true
    bOverrideAngularDamping=true
    bUseGravityOverride=true
    gravityScale=0.35
    simulationSpace=RootRelative
    alphaBoolBlend {blendInTime=0.2, blendOption=Linear, blendOutTime=0.2}
    boundBone {boneName=dyn_ribbon_tape_chest_B_l_2}
    componentLinearAccScale=<1 1 1>
    componentLinearVelScale=<1 1 1>
    gravityOverride {z=-980}
    physicsBodyDefinitions [{
      boundBone {boneName=dyn_ribbon_tape_chest_B_l_2}
      boxExtents=<15 10 10>
      constraintSetup {angularLimitsMax=<5 0 30>, angularLimitsMin=<-5 -45 -30>, angularTarget=<1 0 0>}
      localJointOffset=<3 0 0>
     }]
    simSpaceSettings {maxAngularAcceleration=35, maxAngularVelocity=7, simSpaceAngularAlpha=0.7}
   }
   n50 = Anim Dynamics - Bone: dyn_front_shirt_r(ComponentPose=@n49, bAlphaBoolEnabled=True, ExternalForce="0,0,0") {
    alphaInputType=Bool
    angularDampingOverride=1
    angularSpringConstant=10000
    bAngularSpring=true
    bOverrideAngularDamping=true
    bUseGravityOverride=true
    gravityScale=0.35
    simulationSpace=RootRelative
    alphaBoolBlend {blendInTime=0.2, blendOption=Linear, blendOutTime=0.2}
    boundBone {boneName=dyn_front_shirt_r}
    componentLinearAccScale=<1 1 1>
    componentLinearVelScale=<1 1 1>
    gravityOverride {z=-980}
    physicsBodyDefinitions [{
      boundBone {boneName=dyn_front_shirt_r}
      boxExtents=<15 15 15>
      constraintSetup {angularLimitsMax=<5 5 10>, angularLimitsMin=<-5 -10 -10>, angularTarget=<1 0 0>}
      localJointOffset=<3 0 0>
     }]
    simSpaceSettings {maxAngularAcceleration=35, maxAngularVelocity=7, simSpaceAngularAlpha=0.7}
   }
   n51 = Anim Dynamics - Bone: dyn_front_shirt_l(ComponentPose=@n50, bAlphaBoolEnabled=True, ExternalForce="0,0,0") {
    alphaInputType=Bool
    angularDampingOverride=1
    angularSpringConstant=10000
    bAngularSpring=true
    bOverrideAngularDamping=true
    bUseGravityOverride=true
    gravityScale=0.35
    simulationSpace=RootRelative
    alphaBoolBlend {blendInTime=0.2, blendOption=Linear, blendOutTime=0.2}
    boundBone {boneName=dyn_front_shirt_l}
    componentLinearAccScale=<1 1 1>
    componentLinearVelScale=<1 1 1>
    gravityOverride {z=-980}
    physicsBodyDefinitions [{
      boundBone {boneName=dyn_front_shirt_l}
      boxExtents=<15 15 15>
      constraintSetup {angularLimitsMax=<5 5 10>, angularLimitsMin=<-5 -10 -10>, angularTarget=<1 0 0>}
      localJointOffset=<3 0 0>
     }]
    simSpaceSettings {maxAngularAcceleration=35, maxAngularVelocity=7, simSpaceAngularAlpha=0.7}
   }
   n52 = Anim Dynamics - Bone: dyn_bk_shirt_r(ComponentPose=@n51, bAlphaBoolEnabled=True, ExternalForce="0,0,0") {
    alphaInputType=Bool
    angularDampingOverride=0.9999
    angularSpringConstant=500
    bAngularSpring=true
    bOverrideAngularDamping=true
    bUseGravityOverride=true
    gravityScale=0.35
    simulationSpace=RootRelative
    alphaBoolBlend {blendInTime=0.2, blendOption=Linear, blendOutTime=0.2}
    boundBone {boneName=dyn_bk_shirt_r}
    componentLinearAccScale=<1 1 1>
    componentLinearVelScale=<1 1 1>
    gravityOverride {z=-980}
    physicsBodyDefinitions [{
      boundBone {boneName=dyn_bk_shirt_r}
      boxExtents=<15 10 10>
      constraintSetup {angularLimitsMax=<5 0 10>, angularLimitsMin=<-5 -15 -10>, angularTarget=<1 0 0>}
      localJointOffset=<3 0 0>
     }]
    simSpaceSettings {maxAngularAcceleration=35, maxAngularVelocity=7, simSpaceAngularAlpha=0.7}
   }
   n53 = Anim Dynamics - Bone: dyn_bk_shirt_l(ComponentPose=@n52, bAlphaBoolEnabled=True, ExternalForce="0,0,0") {
    alphaInputType=Bool
    angularDampingOverride=0.9999
    angularSpringConstant=500
    bAngularSpring=true
    bOverrideAngularDamping=true
    bUseGravityOverride=true
    gravityScale=0.35
    simulationSpace=RootRelative
    alphaBoolBlend {blendInTime=0.2, blendOption=Linear, blendOutTime=0.2}
    boundBone {boneName=dyn_bk_shirt_l}
    componentLinearAccScale=<1 1 1>
    componentLinearVelScale=<1 1 1>
    gravityOverride {z=-980}
    physicsBodyDefinitions [{
      boundBone {boneName=dyn_bk_shirt_l}
      boxExtents=<15 10 10>
      constraintSetup {angularLimitsMax=<5 0 10>, angularLimitsMin=<-5 -15 -10>, angularTarget=<1 0 0>}
      localJointOffset=<3 0 0>
     }]
    simSpaceSettings {maxAngularAcceleration=35, maxAngularVelocity=7, simSpaceAngularAlpha=0.7}
   }
   n54 = Anim Dynamics - Bone: dyn_bk_shirt_c(ComponentPose=@n53, bAlphaBoolEnabled=True, ExternalForce="0,0,0") {
    alphaInputType=Bool
    angularDampingOverride=0.9999
    angularSpringConstant=500
    bAngularSpring=true
    bOverrideAngularDamping=true
    bUseGravityOverride=true
    gravityScale=0.35
    simulationSpace=RootRelative
    alphaBoolBlend {blendInTime=0.2, blendOption=Linear, blendOutTime=0.2}
    boundBone {boneName=dyn_bk_shirt_c}
    componentLinearAccScale=<1 1 1>
    componentLinearVelScale=<1 1 1>
    gravityOverride {z=-980}
    physicsBodyDefinitions [{
      boundBone {boneName=dyn_bk_shirt_c}
      boxExtents=<15 10 10>
      constraintSetup {angularLimitsMax=<5 0 10>, angularLimitsMin=<-5 -15 -10>, angularTarget=<1 0 0>}
      localJointOffset=<3 0 0>
     }]
    simSpaceSettings {maxAngularAcceleration=35, maxAngularVelocity=7, simSpaceAngularAlpha=0.7}
   }
   n55 = Anim Dynamics - Bone: dyn_side_shirt_r(ComponentPose=@n54, bAlphaBoolEnabled=True, ExternalForce="0,0,0") {
    alphaInputType=Bool
    angularDampingOverride=0.9999
    angularSpringConstant=500
    bAngularSpring=true
    bOverrideAngularDamping=true
    bUseGravityOverride=true
    gravityScale=0.35
    simulationSpace=RootRelative
    alphaBoolBlend {blendInTime=0.2, blendOption=Linear, blendOutTime=0.2}
    boundBone {boneName=dyn_side_shirt_r}
    componentLinearAccScale=<1 1 1>
    componentLinearVelScale=<1 1 1>
    gravityOverride {z=-980}
    physicsBodyDefinitions [{
      boundBone {boneName=dyn_side_shirt_r}
      boxExtents=<15 10 10>
      constraintSetup {angularLimitsMax=<5 5 10>, angularLimitsMin=<-5 -10 -10>, angularTarget=<1 0 0>}
      localJointOffset=<3 0 0>
     }]
    simSpaceSettings {maxAngularAcceleration=35, maxAngularVelocity=7, simSpaceAngularAlpha=0.7}
   }
   n56 = Anim Dynamics - Bone: dyn_side_shirt_l(ComponentPose=@n55, bAlphaBoolEnabled=True, ExternalForce="0,0,0") {
    alphaInputType=Bool
    angularDampingOverride=0.9999
    angularSpringConstant=500
    bAngularSpring=true
    bOverrideAngularDamping=true
    bUseGravityOverride=true
    gravityScale=0.35
    simulationSpace=RootRelative
    alphaBoolBlend {blendInTime=0.2, blendOption=Linear, blendOutTime=0.2}
    boundBone {boneName=dyn_side_shirt_l}
    componentLinearAccScale=<1 1 1>
    componentLinearVelScale=<1 1 1>
    gravityOverride {z=-980}
    physicsBodyDefinitions [{
      boundBone {boneName=dyn_side_shirt_l}
      boxExtents=<15 10 10>
      constraintSetup {angularLimitsMax=<5 5 10>, angularLimitsMin=<-5 -10 -10>, angularTarget=<1 0 0>}
      localJointOffset=<3 0 0>
     }]
    simSpaceSettings {maxAngularAcceleration=35, maxAngularVelocity=7, simSpaceAngularAlpha=0.7}
   }
   n57 = Anim Dynamics - Bone: dyn_ribbon_arm_r_r(ComponentPose=@n56, bAlphaBoolEnabled=True, ExternalForce="0,0,0") {
    alphaInputType=Bool
    angularDampingOverride=1
    angularSpringConstant=10000
    bAngularSpring=true
    bOverrideAngularDamping=true
    bUseGravityOverride=true
    gravityScale=0.35
    simulationSpace=RootRelative
    alphaBoolBlend {blendInTime=0.2, blendOption=Linear, blendOutTime=0.2}
    boundBone {boneName=dyn_ribbon_arm_r_r}
    componentLinearAccScale=<1 1 1>
    componentLinearVelScale=<1 1 1>
    gravityOverride {z=-980}
    physicsBodyDefinitions [{
      boundBone {boneName=dyn_ribbon_arm_r_r}
      boxExtents=<15 15 15>
      constraintSetup {angularLimitsMax=<10 7.5 10>, angularLimitsMin=<-10 7.5 -10>, angularTarget=<1 0 0>}
      localJointOffset=<3 0 0>
     }]
    simSpaceSettings {maxAngularAcceleration=35, maxAngularVelocity=7, simSpaceAngularAlpha=0.7}
   }
   n58 = Anim Dynamics - Bone: dyn_ribbon_arm_r_l(ComponentPose=@n57, bAlphaBoolEnabled=True, ExternalForce="0,0,0") {
    alphaInputType=Bool
    angularDampingOverride=1
    angularSpringConstant=10000
    bAngularSpring=true
    bOverrideAngularDamping=true
    bUseGravityOverride=true
    gravityScale=0.35
    simulationSpace=RootRelative
    alphaBoolBlend {blendInTime=0.2, blendOption=Linear, blendOutTime=0.2}
    boundBone {boneName=dyn_ribbon_arm_r_l}
    componentLinearAccScale=<1 1 1>
    componentLinearVelScale=<1 1 1>
    gravityOverride {z=-980}
    physicsBodyDefinitions [{
      boundBone {boneName=dyn_ribbon_arm_r_l}
      boxExtents=<15 15 15>
      constraintSetup {angularLimitsMax=<10 7.5 10>, angularLimitsMin=<-10 -7.5 -10>, angularTarget=<1 0 0>}
      localJointOffset=<3 0 0>
     }]
    simSpaceSettings {maxAngularAcceleration=35, maxAngularVelocity=7, simSpaceAngularAlpha=0.7}
   }
   n59 = Anim Dynamics - Bone: dyn_ribbon_tape_arm_r(ComponentPose=@n58, bAlphaBoolEnabled=True, ExternalForce="0,0,0") {
    alphaInputType=Bool
    angularDampingOverride=0.9999
    angularSpringConstant=500
    bAngularSpring=true
    bOverrideAngularDamping=true
    bUseGravityOverride=true
    gravityScale=0.35
    simulationSpace=RootRelative
    alphaBoolBlend {blendInTime=0.2, blendOption=Linear, blendOutTime=0.2}
    boundBone {boneName=dyn_ribbon_tape_arm_r}
    componentLinearAccScale=<1 1 1>
    componentLinearVelScale=<1 1 1>
    gravityOverride {z=-980}
    physicsBodyDefinitions [{
      boundBone {boneName=dyn_ribbon_tape_arm_r}
      boxExtents=<15 10 10>
      constraintSetup {angularLimitsMax=<5 0 30>, angularLimitsMin=<-5 -45 -30>, angularTarget=<1 0 0>}
      localJointOffset=<3 0 0>
     }]
    simSpaceSettings {maxAngularAcceleration=35, maxAngularVelocity=7, simSpaceAngularAlpha=0.7}
   }
   n60 = "Transform (Modify) Bone - Bone: dyn_ribbon_tape_pelvis_l_1"(
    ComponentPose=@n59
    Alpha=Get Custom Idle is Playing Float() -> (Custom Idle is Playing Float:double)
    Translation="0,0,0"
    Rotation="-7,0,15"
    Scale="1,1,1"
   ) {
    rotationMode=BMM_Additive
    alphaBoolBlend {blendOption=Linear}
    boneToModify {boneName=dyn_ribbon_tape_pelvis_l_1}
    rotation {pitch=-7, roll=15}
   }
   Output Pose(Result=Blend Poses by bool(BlendPose_0=Component To Local(ComponentPose=@n60), BlendPose_1=Component To Local(ComponentPose=@n32)) {blendType=Linear}) {layerGroup=None}
   n61 = Layered blend per bone(
    BasePose=Copy Pose From Mesh() {bUseAttachedParent=true}
    BlendPoses_0=Slot 'DefaultSlot'()
    BlendWeights_0=Get Custom Idle is Playing Float()
   ) {layerSetup [{branchFilters [
       {boneName=upperarm_r}
       {boneName=upperarm_twist_01_r}
       {boneName=upperarm_twist_02_r}
       {boneName=lowerarm_r}
       {boneName=lowerarm_twist_01_r}
       {boneName=lowerarm_twist_02_r}
       {boneName=hand_r}
       {boneName=upperarm_l}
       {boneName=upperarm_twist_01_l}
       {boneName=upperarm_twist_02_l}
       {boneName=lowerarm_l}
       {boneName=lowerarm_twist_01_l}
       {boneName=lowerarm_twist_02_l}
       {boneName=hand_l}
      ]}]}
   n62 = Control Rig!(Source=@n61, Alpha, playingEmote:bool=Get bIsPlayingEmote()) -> (Pose) {
    alphaBoolBlend {blendOption=Linear}
    controlRigAssetReference {blueprintRigClass=&/F_MED_TangyRadishFlame_CtrlRig.F_MED_TangyRadishFlame_CtrlRig_C}
   }
   Save cached pose 'ctrlRig_cache'(Pose=@n62) {cacheName=ctrlRig_cache}
   n63 = RigidBody!(
    ComponentPose=Local To Component(LocalPose=Use cached pose 'ctrlRig_cache'())
    Alpha
    OverrideWorldGravity:Vector=Get GravityOverride()
    ExternalForce
    SimSpaceSettings:Sim Space Settings Structure=(WorldAlpha=0,VelocityScaleZ=1,DampingAlpha=1,MaxLinearVelocity=10000,MaxAngularVelocity=10000,MaxLinearAcceleration=10000,MaxAngularAcceleration=10000,ExternalLinearDragV=(X=0,Y=0,Z=0),ExternalLinearVelocity=(X=0,Y=0,Z=0),ExternalAngularVelocity=(X=0,Y=0,Z=0))
   ) -> (Pose) {
    bClampLinearTranslationLimitToRefPose=true
    bFreezeIncomingPoseOnStart=true
    bOverrideWorldGravity=true
    bUseExternalClothCollision=true
    overridePhysicsAsset=&/F_MED_TangyRadishFlame_Physics
    simulationSpace=BaseBoneSpace
    alphaBoolBlend {blendOption=Linear}
    baseBoneRef {boneName=pelvis}
    componentAppliedLinearAccClamp=<0 0 0>
    overrideWorldGravity {z=-980}
    simSpaceSettings {maxAngularAcceleration=100, maxAngularVelocity=4, maxLinearAcceleration=700, maxLinearVelocity=300, velocityScaleZ=0.5, worldAlpha=1}
   }
   n64 = Anim Dynamics - Bone: dyn_front_skirt_l!(ComponentPose=@n63, bAlphaBoolEnabled:bool=True, ExternalForce="0,0,0") -> (Pose) {
    alphaInputType=Bool
    angularDampingOverride=1
    angularSpringConstant=10000
    bAngularSpring=true
    bOverrideAngularDamping=true
    bUseGravityOverride=true
    gravityScale=0.35
    simulationSpace=RootRelative
    alphaBoolBlend {blendInTime=0.2, blendOption=Linear, blendOutTime=0.2}
    boundBone {boneName=dyn_front_skirt_l}
    componentLinearAccScale=<1 1 1>
    componentLinearVelScale=<1 1 1>
    gravityOverride {z=-980}
    physicsBodyDefinitions [{
      boundBone {boneName=dyn_front_skirt_l}
      boxExtents=<10 10 10>
      constraintSetup {angularLimitsMax=<5 0 10>, angularLimitsMin=<-5 -5 -10>, angularTarget=<1 0 0>}
      localJointOffset=<3 0 0>
     }]
   }
   n65 = Anim Dynamics - Bone: dyn_front_skirt_r!(ComponentPose=@n64, bAlphaBoolEnabled:bool=True, ExternalForce="0,0,0") -> (Pose) {
    alphaInputType=Bool
    angularDampingOverride=1
    angularSpringConstant=10000
    bAngularSpring=true
    bOverrideAngularDamping=true
    bUseGravityOverride=true
    gravityScale=0.35
    simulationSpace=RootRelative
    alphaBoolBlend {blendInTime=0.2, blendOption=Linear, blendOutTime=0.2}
    boundBone {boneName=dyn_front_skirt_r}
    componentLinearAccScale=<1 1 1>
    componentLinearVelScale=<1 1 1>
    gravityOverride {z=-980}
    physicsBodyDefinitions [{
      boundBone {boneName=dyn_front_skirt_r}
      boxExtents=<10 10 10>
      constraintSetup {angularLimitsMax=<5 0 10>, angularLimitsMin=<-5 -5 -10>, angularTarget=<1 0 0>}
      localJointOffset=<3 0 0>
     }]
   }
   n66 = Anim Dynamics - Bone: dyn_side_skirt_l!(ComponentPose=@n65, bAlphaBoolEnabled:bool=True, ExternalForce="0,0,0") -> (Pose) {
    alphaInputType=Bool
    angularDampingOverride=0.99
    angularSpringConstant=500
    bAngularSpring=true
    bOverrideAngularDamping=true
    bUseGravityOverride=true
    gravityScale=0.35
    simulationSpace=RootRelative
    alphaBoolBlend {blendInTime=0.2, blendOption=Linear, blendOutTime=0.2}
    boundBone {boneName=dyn_side_skirt_l}
    componentAppliedLinearAccClamp=<3000 3000 3000>
    componentLinearAccScale=<0.5 0.5 0.5>
    componentLinearVelScale=<0.5 0.5 0.5>
    gravityOverride {z=-980}
    physicsBodyDefinitions [{
      boundBone {boneName=dyn_side_skirt_l}
      boxExtents=<15 15 15>
      constraintSetup {angularLimitsMax=<0 5 20>, angularLimitsMin=<0 -40 -20>, angularTarget=<1 0 0>}
      localJointOffset=<3 0 0>
     }]
   }
   n67 = Anim Dynamics - Bone: dyn_side_skirt_r!(ComponentPose=@n66, bAlphaBoolEnabled:bool=True, ExternalForce="0,0,0") -> (Pose) {
    alphaInputType=Bool
    angularDampingOverride=0.99
    angularSpringConstant=500
    bAngularSpring=true
    bOverrideAngularDamping=true
    bUseGravityOverride=true
    gravityScale=0.35
    simulationSpace=RootRelative
    alphaBoolBlend {blendInTime=0.2, blendOption=Linear, blendOutTime=0.2}
    boundBone {boneName=dyn_side_skirt_r}
    componentAppliedLinearAccClamp=<3000 3000 3000>
    componentLinearAccScale=<0.5 0.5 0.5>
    componentLinearVelScale=<0.5 0.5 0.5>
    gravityOverride {z=-980}
    physicsBodyDefinitions [{
      boundBone {boneName=dyn_side_skirt_r}
      boxExtents=<15 15 15>
      constraintSetup {angularLimitsMax=<0 5 20>, angularLimitsMin=<0 -40 -20>, angularTarget=<1 0 0>}
      localJointOffset=<3 0 0>
     }]
   }
   n68 = Anim Dynamics - Bone: dyn_bk_skirt_l!(ComponentPose=@n67, bAlphaBoolEnabled:bool=True, ExternalForce="0,0,0") -> (Pose) {
    alphaInputType=Bool
    angularDampingOverride=0.99
    angularSpringConstant=500
    bAngularSpring=true
    bOverrideAngularDamping=true
    bUseGravityOverride=true
    gravityScale=0.35
    simulationSpace=RootRelative
    alphaBoolBlend {blendInTime=0.2, blendOption=Linear, blendOutTime=0.2}
    boundBone {boneName=dyn_bk_skirt_l}
    componentAppliedLinearAccClamp=<3000 3000 3000>
    componentLinearAccScale=<0.5 0.5 0.5>
    componentLinearVelScale=<0.5 0.5 0.5>
    gravityOverride {z=-980}
    physicsBodyDefinitions [{
      boundBone {boneName=dyn_bk_skirt_l}
      boxExtents=<15 15 10>
      constraintSetup {angularLimitsMax=<10 5 8>, angularLimitsMin=<-10 -40 -8>, angularTarget=<1 0 0>}
      localJointOffset=<3 0 0>
     }]
   }
   n69 = Anim Dynamics - Bone: dyn_bk_skirt_r!(ComponentPose=@n68, bAlphaBoolEnabled:bool=True, ExternalForce="0,0,0") -> (Pose) {
    alphaInputType=Bool
    angularDampingOverride=0.99
    angularSpringConstant=500
    bAngularSpring=true
    bOverrideAngularDamping=true
    bUseGravityOverride=true
    gravityScale=0.35
    simulationSpace=RootRelative
    alphaBoolBlend {blendInTime=0.2, blendOption=Linear, blendOutTime=0.2}
    boundBone {boneName=dyn_bk_skirt_r}
    componentAppliedLinearAccClamp=<3000 3000 3000>
    componentLinearAccScale=<0.5 0.5 0.5>
    componentLinearVelScale=<0.5 0.5 0.5>
    gravityOverride {z=-980}
    physicsBodyDefinitions [{
      boundBone {boneName=dyn_bk_skirt_r}
      boxExtents=<15 15 10>
      constraintSetup {angularLimitsMax=<10 5 8>, angularLimitsMin=<-10 -40 -8>, angularTarget=<1 0 0>}
      localJointOffset=<3 0 0>
     }]
   }
   Save cached pose 'skirtDyn_cache'(Pose=Component To Local(ComponentPose=@n69)) {cacheName=skirtDyn_cache}
  ]}
 EventGraph {nodes [
   n1 = Event BlueprintUpdateAnimation()
   > Set DeltaTime X(DeltaTime X:double=@n1.DeltaTimeX) -> (Output_Get:double)
   > Sequence()
   > n2 = "Gravity Override State Function (Nativized)"!(
    execute
    CharacterPartAnimInstance:Custom Character Part Anim Instance Object Reference=Self-Reference()
    GravityOverrideParameterSet:Fort Gravity Override Parameter Set Structure=Get Gravity Override Params() -> (Gravity Override Params:Fort Gravity Override Parameter Set Structure)
    RigidBodyAnimNodeParameterSet:Fort Rigid Body Anim Node Parameter Set Structure=Get RBAN Params() -> (RBAN Params:Fort Rigid Body Anim Node Parameter Set Structure)
    DebugDraw:bool=false
    GameTimeInSeconds:float=GetGameTimeInSeconds()
    DebugJointName:Name=root
    StateLogicLayerName:Name=none
   ) -> (
    then
    OutGravityOverride:Vector
    OutCurrentState:Name=None
    OutGravityOverrideParameters:Fort Gravity Override Parameters Structure
    OutRigidBodyAnimNodeParameters:Fort Rigid Body Anim Node Parameters Structure
   )
   > Set GravityOverride(GravityOverride=@n2)
   > Set RBAN Settings Out(RBAN Settings Out=@n2.OutRigidBodyAnimNodeParameters)
   n3 = Get bIsPlayingEmote()
   > Set isPlayEmote(isPlayEmote:double=FInterpTo(
     Current=Get isPlayEmote() -> (isPlayEmote:double)
     Target="To Float (Boolean)"(InBool=@n3)
     DeltaTime=Get DeltaTime X() -> (DeltaTime X:double)
     InterpSpeed=2
    )) -> (Output_Get:double)
   > n4 = Branch(Condition=Get bIsInFrontEnd())
   > Set Custom Idle is Playing Float(Custom Idle is Playing Float:double=GetCurveValue(self=GetOwnerFrontEndAnimInstance(), CurveName=Frontend_Idle_Bypass)) -> (Output_Get:double)
   Set Custom Idle is Playing Float(execute=@n4.else, Custom Idle is Playing Float=0)
  ]}
]
variables [
 Custom Idle is Playing Float {type=number}
 DeltaTime X {type=number}
 Gravity Override Params {
  type=FortGravityOverrideParameterSet
  default=(Skydive_GoingUp=(GravityOverride=(WindFrequency=(X=1.500000,Y=7.000000,Z=8.000000),WindAmplitude=(X=200.000000,Y=-500.000000,Z=400.000000),GravityOverride=(X=0.000000,Y=-300.000000,Z=100.000000),JointName="root")),Skydive_FloatingDown=(GravityOverride=(WindFrequency=(X=1.500000,Y=7.000000,Z=8.000000),WindAmplitude=(X=200.000000,Y=-500.000000,Z=400.000000),WindOffset=(X=0.000000,Y=-600.000000,Z=400.000000),GravityOverride=(X=0.000000,Y=-700.000000,Z=600.000000),JointName="root")),Skydive_DivingDown=(GravityOverride=(WindFrequency=(X=1.500000,Y=12.000000,Z=10.000000),WindAmplitude=(X=50.000000,Y=300.000000,Z=100.000000),GravityOverride=(X=-1500.000000,Y=0.000000,Z=0.000000),JointName="pelvis")),Skydive_Parachuting=(GravityOverride=(WindFrequency=(X=1.500000,Y=7.000000,Z=8.000000),WindAmplitude=(X=200.000000,Y=-500.000000,Z=400.000000),GravityOverride=(X=0.000000,Y=-800.000000,Z=-980.000000),JointName="root")),OnGround_Standing=(GravityOverride=(GravityOverride=(X=0.000000,Y=0.000000,Z=-980.000000),JointName="root")),OnGround_Moving=(GravityOverride=(GravityOverride=(X=0.000000,Y=0.000000,Z=-980.000000),JointName="root")),OnGround_Moving_HighRotationRate=(GravityOverride=(GravityOverride=(X=0.000000,Y=0.000000,Z=-980.000000),JointName="root")),OnGround_Crouch_Moving=(GravityOverride=(GravityOverride=(X=0.000000,Y=0.000000,Z=-980.000000),JointName="root")),Emote_or_Melee=(GravityOverride=(GravityOverride=(X=0.000000,Y=0.000000,Z=-980.000000),JointName="root")),Swimming=(GravityOverride=(GravityOverride=(X=-800.000000,Y=0.000000,Z=0.000000),JointName="pelvis")),Front_End=(GravityOverride=(GravityOverride=(X=0.000000,Y=0.000000,Z=-980.000000),JointName="root")),DBNO=(GravityOverride=(GravityOverride=(X=-980.000000,Y=400.000000,Z=0.000000),JointName="pelvis")),DBNO_Carried=(GravityOverride=(GravityOverride=(X=-980.000000,Y=500.000000,Z=-500.000000),JointName="pelvis")),OnGround_Falling=(GravityOverride=(GravityOverride=(X=0.000000,Y=0.000000,Z=-980.000000),JointName="root")),Ziplining=(GravityOverride=(GravityOverride=(X=0.000000,Y=0.000000,Z=-980.000000),JointName="root")),Skydive_Surfing=(GravityOverride=(WindFrequency=(X=1.500000,Y=7.000000,Z=8.000000),WindAmplitude=(X=200.000000,Y=-500.000000,Z=400.000000),GravityOverride=(X=500.000000,Y=-300.000000,Z=-980.000000),JointName="root")),Riding_Standing=(GravityOverride=(GravityOverride=(X=0.000000,Y=-350.000000,Z=-980.000000),JointName="root")),Riding_Moving=(GravityOverride=(GravityOverride=(X=0.000000,Y=-450.000000,Z=-450.000000),JointName="root")),Riding_Moving_Targeting=(GravityOverride=(GravityOverride=(X=0.000000,Y=-200.000000,Z=-980.000000),JointName="root")),Tactical_Sprint=(GravityOverride=(GravityOverride=(X=0.000000,Y=-450.000000,Z=-500.000000),JointName="root")),Rail_Grinding=(GravityOverride=(GravityOverride=(X=0.000000,Y=0.000000,Z=-980.000000),JointName="root")),)
 }
 GravityOverride {type=Vector}
 isPlayEmote {type=number}
 RBAN Params {
  type=FortRigidBodyAnimNodeParameterSet
  default=(Skydive_GoingUp=(RigidBodyAnimNode=(ComponentLinearAccScale=(X=0.000000,Y=-0.500000,Z=0.000000),ComponentLinearVelScale=(X=0.000000,Y=-0.500000,Z=0.000000),SimSpaceSettings=(WorldAlpha=1.000000,VelocityScaleZ=0.000000,DampingAlpha=0.000000,MaxLinearVelocity=300.000000,MaxAngularVelocity=4.000000,MaxLinearAcceleration=1200.000000,MaxAngularAcceleration=16.000000,ExternalLinearDragV=(X=0.000000,Y=0.000000,Z=-1.000000)))),Skydive_FloatingDown=(RigidBodyAnimNode=(ComponentLinearAccScale=(X=0.000000,Y=0.500000,Z=0.000000),ComponentLinearVelScale=(X=0.000000,Y=0.500000,Z=0.000000),SimSpaceSettings=(WorldAlpha=1.000000,VelocityScaleZ=0.500000,DampingAlpha=0.000000,MaxLinearVelocity=700.000000,MaxAngularVelocity=8.000000,MaxLinearAcceleration=2800.000000,MaxAngularAcceleration=32.000000,ExternalLinearDragV=(X=0.500000,Y=0.500000,Z=0.500000)))),Skydive_DivingDown=(RigidBodyAnimNode=(ComponentLinearAccScale=(X=0.100000,Y=0.000000,Z=0.000000),ComponentLinearVelScale=(X=0.100000,Y=0.000000,Z=0.000000),SimSpaceSettings=(WorldAlpha=1.000000,VelocityScaleZ=0.500000,DampingAlpha=0.000000,MaxLinearVelocity=700.000000,MaxAngularVelocity=8.000000,MaxLinearAcceleration=2800.000000,MaxAngularAcceleration=32.000000,ExternalLinearDragV=(X=0.500000,Y=0.500000,Z=0.500000)))),Skydive_Parachuting=(RigidBodyAnimNode=(ComponentLinearAccScale=(X=0.000000,Y=0.100000,Z=0.300000),ComponentLinearVelScale=(X=0.000000,Y=0.100000,Z=0.300000),SimSpaceSettings=(WorldAlpha=1.000000,VelocityScaleZ=0.500000,DampingAlpha=0.000000,MaxLinearVelocity=700.000000,MaxAngularVelocity=8.000000,MaxLinearAcceleration=2800.000000,MaxAngularAcceleration=32.000000,ExternalLinearDragV=(X=0.500000,Y=0.500000,Z=0.500000)))),OnGround_Standing=(RigidBodyAnimNode=(ComponentLinearAccScale=(X=0.300000,Y=0.000000,Z=0.000000),ComponentLinearVelScale=(X=0.300000,Y=0.000000,Z=0.000000),SimSpaceSettings=(WorldAlpha=1.000000,VelocityScaleZ=0.500000,DampingAlpha=0.000000,MaxLinearVelocity=300.000000,MaxAngularVelocity=6.000000,MaxLinearAcceleration=1200.000000,MaxAngularAcceleration=24.000000,ExternalLinearDragV=(X=0.200000,Y=0.200000,Z=0.200000)))),OnGround_Moving=(RigidBodyAnimNode=(ComponentLinearAccScale=(X=0.300000,Y=0.000000,Z=0.000000),ComponentLinearVelScale=(X=0.300000,Y=0.000000,Z=0.000000),SimSpaceSettings=(WorldAlpha=1.000000,VelocityScaleZ=0.500000,DampingAlpha=0.000000,MaxLinearVelocity=200.000000,MaxAngularVelocity=3.000000,MaxLinearAcceleration=800.000000,MaxAngularAcceleration=12.000000,ExternalLinearDragV=(X=0.200000,Y=0.200000,Z=0.200000)))),OnGround_Moving_HighRotationRate=(RigidBodyAnimNode=(ComponentLinearAccScale=(X=0.300000,Y=0.000000,Z=0.000000),ComponentLinearVelScale=(X=0.300000,Y=0.000000,Z=0.000000),SimSpaceSettings=(WorldAlpha=1.000000,VelocityScaleZ=0.500000,DampingAlpha=0.000000,MaxLinearVelocity=300.000000,MaxAngularVelocity=6.000000,MaxLinearAcceleration=1200.000000,MaxAngularAcceleration=24.000000,ExternalLinearDragV=(X=0.200000,Y=0.200000,Z=0.200000)))),OnGround_Crouch_Moving=(RigidBodyAnimNode=(SimSpaceSettings=(WorldAlpha=1.000000,VelocityScaleZ=0.500000,DampingAlpha=0.000000,MaxLinearVelocity=400.000000,MaxAngularVelocity=8.000000,MaxLinearAcceleration=3200.000000,MaxAngularAcceleration=32.000000,ExternalLinearDragV=(X=0.500000,Y=0.500000,Z=0.500000)))),Emote_or_Melee=(RigidBodyAnimNode=(SimSpaceSettings=(WorldAlpha=1.000000,VelocityScaleZ=0.500000,DampingAlpha=0.000000,MaxLinearVelocity=700.000000,MaxAngularVelocity=8.000000,MaxLinearAcceleration=2800.000000,MaxAngularAcceleration=32.000000,ExternalLinearDragV=(X=0.500000,Y=0.500000,Z=0.500000)))),Swimming=(RigidBodyAnimNode=(SimSpaceSettings=(WorldAlpha=1.000000,VelocityScaleZ=0.500000,DampingAlpha=0.000000,MaxLinearVelocity=700.000000,MaxAngularVelocity=8.000000,MaxLinearAcceleration=2800.000000,MaxAngularAcceleration=32.000000,ExternalLinearDragV=(X=0.500000,Y=0.500000,Z=0.500000)))),Front_End=(RigidBodyAnimNode=(SimSpaceSettings=(WorldAlpha=1.000000,VelocityScaleZ=0.500000,DampingAlpha=0.000000,MaxLinearVelocity=200.000000,MaxAngularVelocity=2.500000,MaxLinearAcceleration=800.000000,MaxAngularAcceleration=10.000000))),DBNO=(RigidBodyAnimNode=(SimSpaceSettings=(WorldAlpha=1.000000,VelocityScaleZ=0.500000,DampingAlpha=0.000000,MaxLinearVelocity=700.000000,MaxAngularVelocity=8.000000,MaxLinearAcceleration=2800.000000,MaxAngularAcceleration=32.000000,ExternalLinearDragV=(X=0.500000,Y=0.500000,Z=0.500000)))),DBNO_Carried=(RigidBodyAnimNode=(SimSpaceSettings=(WorldAlpha=1.000000,VelocityScaleZ=0.500000,DampingAlpha=0.000000,MaxLinearVelocity=700.000000,MaxAngularVelocity=8.000000,MaxLinearAcceleration=2800.000000,MaxAngularAcceleration=32.000000,ExternalLinearDragV=(X=0.500000,Y=0.500000,Z=0.500000)))),OnGround_Falling=(RigidBodyAnimNode=(SimSpaceSettings=(WorldAlpha=1.000000,VelocityScaleZ=0.300000,DampingAlpha=0.000000,MaxLinearVelocity=700.000000,MaxAngularVelocity=8.000000,MaxLinearAcceleration=2800.000000,MaxAngularAcceleration=32.000000))),Ziplining=(RigidBodyAnimNode=(SimSpaceSettings=(WorldAlpha=1.000000,VelocityScaleZ=0.500000,DampingAlpha=0.000000,MaxLinearVelocity=700.000000,MaxAngularVelocity=8.000000,MaxLinearAcceleration=2800.000000,MaxAngularAcceleration=32.000000,ExternalLinearDragV=(X=0.500000,Y=0.500000,Z=0.500000)))),Skydive_Surfing=(RigidBodyAnimNode=(SimSpaceSettings=(WorldAlpha=1.000000,VelocityScaleZ=0.500000,DampingAlpha=0.000000,MaxLinearVelocity=700.000000,MaxAngularVelocity=8.000000,MaxLinearAcceleration=2800.000000,MaxAngularAcceleration=32.000000,ExternalLinearDragV=(X=0.500000,Y=0.500000,Z=0.500000)))),Riding_Standing=(RigidBodyAnimNode=(SimSpaceSettings=(WorldAlpha=1.000000,VelocityScaleZ=0.500000,DampingAlpha=0.000000,MaxLinearVelocity=700.000000,MaxAngularVelocity=8.000000,MaxLinearAcceleration=2800.000000,MaxAngularAcceleration=32.000000,ExternalLinearDragV=(X=0.500000,Y=0.500000,Z=0.500000)))),Riding_Moving=(RigidBodyAnimNode=(SimSpaceSettings=(WorldAlpha=1.000000,VelocityScaleZ=0.500000,DampingAlpha=0.000000,MaxLinearVelocity=700.000000,MaxAngularVelocity=8.000000,MaxLinearAcceleration=2800.000000,MaxAngularAcceleration=32.000000,ExternalLinearDragV=(X=0.500000,Y=0.500000,Z=0.500000)))),Riding_Moving_Targeting=(RigidBodyAnimNode=(SimSpaceSettings=(WorldAlpha=1.000000,VelocityScaleZ=0.500000,DampingAlpha=0.000000,MaxLinearVelocity=700.000000,MaxAngularVelocity=8.000000,MaxLinearAcceleration=2800.000000,MaxAngularAcceleration=32.000000,ExternalLinearDragV=(X=0.500000,Y=0.500000,Z=0.500000)))),Tactical_Sprint=(RigidBodyAnimNode=(ComponentLinearAccScale=(X=0.300000,Y=0.000000,Z=0.000000),ComponentLinearVelScale=(X=0.300000,Y=0.000000,Z=0.000000),SimSpaceSettings=(WorldAlpha=1.000000,VelocityScaleZ=0.100000,DampingAlpha=0.000000,MaxLinearVelocity=200.000000,MaxAngularVelocity=3.000000,MaxLinearAcceleration=800.000000,MaxAngularAcceleration=12.000000,ExternalLinearDragV=(X=0.200000,Y=0.200000,Z=0.200000)))),Rail_Grinding=(RigidBodyAnimNode=(SimSpaceSettings=(WorldAlpha=1.000000,VelocityScaleZ=0.500000,DampingAlpha=0.000000,MaxLinearVelocity=700.000000,MaxAngularVelocity=8.000000,MaxLinearAcceleration=2800.000000,MaxAngularAcceleration=32.000000,ExternalLinearDragV=(X=0.500000,Y=0.500000,Z=0.500000)))),)
 }
 RBAN Settings Out {
  type=FortRigidBodyAnimNodeParameters
  default {simSpaceSettings {
    dampingAlpha=0
    maxAngularAcceleration=16
    maxAngularVelocity=4
    maxLinearAcceleration=800
    maxLinearVelocity=200
    velocityScaleZ=0.5
    worldAlpha=1
   }}
 }
]
