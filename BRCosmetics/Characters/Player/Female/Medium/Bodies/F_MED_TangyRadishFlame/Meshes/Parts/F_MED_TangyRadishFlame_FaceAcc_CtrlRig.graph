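# Control Rig for the FaceAcc mesh part of the F_MED_TangyRadishFlame player cosmetic, which corrects short-hair bones based on head pose.
# A spherical pose reader on the head bone outputs a falloff weight from an asymmetric, offset active region.
# That weight drives a Modify Transforms unit that rotates back-hair base bones, including center and left, carrying their child dyn_ chains along as the head moves.
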ControlRigBlueprint /BRCosmetics/Characters/Player/Female/Medium/Bodies/F_MED_TangyRadishFlame/Meshes/Parts/F_MED_TangyRadishFlame_FaceAcc_CtrlRig
graphs [
 RigVMFunctionLibrary {links [], nodes []}
 RigVMModel {nodes [
   RigUnit_BeginExecution = &RigUnit_BeginExecution()
   RigVMFunction_Sequence = &RigVMFunction_Sequence(ExecuteContext=@RigUnit_BeginExecution.ExecutePin)
   n1 = &RigUnit_SphericalPoseReader(
    ActiveRegionScaleFactors=(PositiveWidth=1,NegativeWidth=0.2,PositiveHeight=1.5,NegativeHeight=1.5)
    ActiveRegionSize=0.1
    Debug=(bDrawDebug=false,bDraw2D=false,bDrawLocalAxes=false,DebugScale=25,DebugSegments=20,DebugThickness=0.25)
    DriverAxis=(X=0,Y=1,Z=0)
    DriverItem=(Type=Bone,Name="head")
    FalloffRegionScaleFactors=(PositiveWidth=3,NegativeWidth=1.8,PositiveHeight=2.5,NegativeHeight=2.5)
    FalloffSize=0.2
    FlipHeightScaling=False
    FlipWidthScaling=False
    OptionalParentItem=(Type=Bone,Name="None")
    RotationOffset=(X=90,Y=0,Z=-55)
    ExecutePin=@RigVMFunction_Sequence.A
   )
   > &RigUnit_ModifyTransforms(
    ItemToModify="((Item=(Type=Bone,Name=\"base_hair_short_bk_c\"),Transform=(Rotation=(X=0,Y=-0.3827,Z=0,W=0.9239),Translation=(X=0,Y=0,Z=0),Scale3D=(X=1,Y=1,Z=1))),(Item=(Type=Bone,Name=\"base_hair_short_bk_l\"),Transform=(Rotation=(X=0,Y=-0.3827,Z=0,W=0.9239),Translation=(X= ...(truncated)"
    WeightMinimum=0
    Weight=@n1.OutputParam
   )
  ]}
]
hierarchy [
 root {loc=<0 0 0>, rot=<0 0 0 1>}
 pelvis {parent=root, loc=<0 -0.5301 98.0571>, rot=<0.0127 0.707 -0.0127 -0.707>}
 spine_01 {parent=pelvis, loc=<6.1208 -0.9328 0>, rot=<0 0 0.034 -0.9994>}
 spine_02 {parent=spine_01, loc=<6.5997 -0.2095 0>, rot=<0 0 0.0238 -0.9997>}
 spine_03 {parent=spine_02, loc=<6.4914 -0.1237 0>, rot=<0 0 0.0149 -0.9999>}
 spine_04 {parent=spine_03, loc=<6.9598 0.0794 0>, rot=<0 0 -0.0132 -0.9999>}
 spine_05 {parent=spine_04, loc=<6.9987 0 0>, rot=<0 0 -0.0015 -1>}
 neck_01 {parent=spine_05, loc=<11.3138 0.1529 0>, rot=<0 0 -0.2003 -0.9797>}
 neck_02 {parent=neck_01, loc=<4.5627 0 0>, rot=<0 0 0.0504 -0.9987>}
 head {parent=neck_02, loc=<5.3914 0 0>, rot=<0 0 0.115 -0.9934>}
 main_hair_short {parent=head, loc=<0 0 0>, rot=<0 0 0 1>}
 base_hair_short_side_r {parent=main_hair_short, loc=<3.248 -1.4431 5.6031>, rot=<0.0368 -0.02 0.9951 -0.0897>}
 dyn_hair_short_side_r_01 {parent=base_hair_short_side_r, loc=<0 0 0>, rot=<0 0 0 1>}
 dyn_hair_short_side_r_02 {parent=dyn_hair_short_side_r_01, loc=<4.6547 0 0>, rot=<0 0 0 1>}
 dyn_hair_short_side_r_03 {parent=dyn_hair_short_side_r_02, loc=<4.2033 -0.0001 0.0002>, rot=<0 0 0 1>}
 base_hair_short_bk_r {parent=main_hair_short, loc=<7.5605 -5.7764 3.5076>, rot=<0.0743 -0.7217 0.6845 -0.0711>}
 dyn_hair_short_bk_r_01 {parent=base_hair_short_bk_r, loc=<0 0 0>, rot=<0 0 0 1>}
 dyn_hair_short_bk_r_02 {parent=dyn_hair_short_bk_r_01, loc=<7.5605 0.0039 -0.051>, rot=<0 0 0 1>}
 dyn_hair_short_bk_r_03 {parent=dyn_hair_short_bk_r_02, loc=<6.6495 0.0009 -0.0033>, rot=<0 0 0 1>}
 dyn_hair_short_bk_r_04 {parent=dyn_hair_short_bk_r_03, loc=<7.7275 -0.0048 0.0543>, rot=<0 0 0 1>}
 base_hair_short_bk_c {parent=main_hair_short, loc=<8.1734 -6.9159 -0.4528>, rot=<0.0535 -0.7235 0.6865 -0.0492>}
 dyn_hair_short_bk_c_01 {parent=base_hair_short_bk_c, loc=<0 0 0>, rot=<0 0 0 1>}
 dyn_hair_short_bk_c_02 {parent=dyn_hair_short_bk_c_01, loc=<8.9153 0.0039 -0.0646>, rot=<0 0.0245 0 -0.9997>}
 dyn_hair_short_bk_c_03 {parent=dyn_hair_short_bk_c_02, loc=<7.0488 -0.0083 -0.188>, rot=<0 -0.0025 0 -1>}
 dyn_hair_short_bk_c_04 {parent=dyn_hair_short_bk_c_03, loc=<9.099 -0.0048 0.0901>, rot=<0 0 0 1>}
 base_hair_short_bk_l {parent=main_hair_short, loc=<7.4836 -5.3736 -4.2613>, rot=<0.0743 -0.7217 0.6845 -0.0711>}
 dyn_hair_short_bk_l_01 {parent=base_hair_short_bk_l, loc=<0 0 0>, rot=<0 0 0 1>}
 dyn_hair_short_bk_l_02 {parent=dyn_hair_short_bk_l_01, loc=<7.5606 0.0039 -0.0511>, rot=<0 0 0 1>}
 dyn_hair_short_bk_l_03 {parent=dyn_hair_short_bk_l_02, loc=<6.6493 0.0009 -0.0031>, rot=<0 0 0 1>}
 dyn_hair_short_bk_l_04 {parent=dyn_hair_short_bk_l_03, loc=<7.7275 -0.0048 0.0543>, rot=<0 0 0 1>}
 base_hair_short_side_l {parent=main_hair_short, loc=<2.9389 -1.4964 -5.8732>, rot=<0.0289 -0.9984 -0.045 0.0176>}
 dyn_hair_short_side_l_01 {parent=base_hair_short_side_l, loc=<0 0 0>, rot=<0 0 0 1>}
 dyn_hair_short_side_l_02 {parent=dyn_hair_short_side_l_01, loc=<4.655 0 0>, rot=<0 0 0 1>}
 dyn_hair_short_side_l_03 {parent=dyn_hair_short_side_l_02, loc=<4.2036 0 0.0002>, rot=<0 0 0 1>}
 base_hair_short_side_fr_r {parent=main_hair_short, loc=<6.1658 3.0242 6.8079>, rot=<-0.0181 -0.0151 0.9996 0.0163>}
 dyn_hair_short_side_fr_r_01 {parent=base_hair_short_side_fr_r, loc=<0 0 0>, rot=<0 0 0 1>}
 dyn_hair_short_side_fr_r_02 {parent=dyn_hair_short_side_fr_r_01, loc=<2.968 0 0>, rot=<0 0 0 1>}
 base_hair_short_side_fr_l {parent=main_hair_short, loc=<5.6697 3.6297 -6.177>, rot=<-0.0026 -0.9988 -0.0444 0.019>}
 dyn_hair_short_side_fr_l_01 {parent=base_hair_short_side_fr_l, loc=<0 0 0>, rot=<0 0 0 1>}
 dyn_hair_short_side_fr_l_02 {parent=dyn_hair_short_side_fr_l_01, loc=<2.5199 0 0>, rot=<0 0 0 1>}
 base_hair_short_front_l {parent=main_hair_short, loc=<7.5196 8.4894 -3.6344>, rot=<-0.0731 -0.8087 -0.583 -0.0265>}
 dyn_hair_short_front_l {parent=base_hair_short_front_l, loc=<0 0 0>, rot=<0 0 0 1>}
 base_hair_short_front_r {parent=main_hair_short, loc=<7.6535 8.3957 3.6249>, rot=<0.0225 0.5636 0.8258 0.002>}
 dyn_hair_short_front_r {parent=base_hair_short_front_r, loc=<0 0 0>, rot=<0 0 0 1>}
 clavicle_l {parent=spine_05, loc=<5.4831 -0.1053 -3.2946>, rot=<0.0465 -0.7289 -0.009 -0.6829>}
 upperarm_l {parent=clavicle_l, loc=<9.8658 -1.2136 -0.4105>, rot=<0.0164 -0.3245 -0.0269 -0.9454>}
 clavicle_r {parent=spine_05, loc=<5.4831 -0.1053 3.2946>, rot=<0.7289 0.0465 -0.6829 0.009>}
 upperarm_r {parent=clavicle_r, loc=<-9.8658 1.2136 0.4105>, rot=<0.0164 -0.3245 -0.0269 -0.9454>}
]
