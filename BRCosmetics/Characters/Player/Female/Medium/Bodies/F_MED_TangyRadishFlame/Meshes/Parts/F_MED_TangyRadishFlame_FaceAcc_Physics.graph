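# Physics asset for the short hair on the TangyRadishFlame female medium character's FaceAcc mesh part.
# Three back chains and two side chains of capsules hang from head-linked base bones, with angular limits that widen down the back chains and stay uniform on the sides.
# Angularly free cross-constraints tie strands together asymmetrically, including a duplicated left-side link.
# Head, neck, upper spine, clavicle and upper-arm capsules provide body collision shapes.
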PhysicsAsset /BRCosmetics/Characters/Player/Female/Medium/Bodies/F_MED_TangyRadishFlame/Meshes/Parts/F_MED_TangyRadishFlame_FaceAcc_Physics
bodies {
 base_hair_short_bk_c [{shapeType=Capsule, length=0.505, radius=0.505, shapeName=base_hair_short_bk_c_capsule, rotation=<rot 0 -90 0>}]
 base_hair_short_bk_l [{shapeType=Capsule, length=0.505, radius=0.505, shapeName=base_hair_short_bk_l_capsule, rotation=<rot 0 -90 0>}]
 base_hair_short_bk_r [{shapeType=Capsule, length=0.505, radius=0.505, shapeName=base_hair_short_bk_r_capsule, rotation=<rot 0 -90 0>}]
 base_hair_short_side_l [{shapeType=Capsule, length=0.505, radius=0.505, shapeName=base_hair_short_side_l_capsule, rotation=<rot 0 -90 0>}]
 base_hair_short_side_r [{shapeType=Capsule, length=0.505, radius=0.505, shapeName=base_hair_short_side_r_capsule, rotation=<rot 0 -90 0>}]
 clavicle_l [{shapeType=Capsule, length=5.9964, radius=4.8458, shapeName=clavicle_l_capsule, center=<4.6227 0 -0.7837>, rotation=<rot -90 0 0>}]
 clavicle_r [{shapeType=Capsule, length=5.8423, radius=5, shapeName=clavicle_r_capsule, center=<-4.6227 0 0.3574>, rotation=<rot -90 0 0>}]
 dyn_hair_short_bk_c_01 [{shapeType=Capsule, length=3, radius=3, center=<7 0 0>, rotation=<rot -90 0 0>}]
 dyn_hair_short_bk_c_02 [{shapeType=Capsule, length=3, radius=2.5, center=<4 0 0>, rotation=<rot -90 90 0>}]
 dyn_hair_short_bk_c_03 [{shapeType=Capsule, length=6, radius=2, center=<6 0 1.5>, rotation=<rot -90 90 0>}]
 dyn_hair_short_bk_c_04 [{shapeType=Capsule, length=3, radius=2, center=<4 0 1>, rotation=<rot -90 0 0>}]
 dyn_hair_short_bk_l_01 [{shapeType=Capsule, length=3, radius=3, center=<7 0 0>, rotation=<rot -90 0 0>}]
 dyn_hair_short_bk_l_02 [{shapeType=Capsule, length=6.6672, radius=3.3328, center=<3.5618 0 0.715>, rotation=<rot -90 90 0>}]
 dyn_hair_short_bk_l_03 [{shapeType=Capsule, length=5, radius=2.5, center=<4 0 1.5>, rotation=<rot -90 0 0>}]
 dyn_hair_short_bk_l_04 [{shapeType=Capsule, length=3, radius=2, center=<4 0 1.0963>, rotation=<rot -90 0 0>}]
 dyn_hair_short_bk_r_01 [{shapeType=Capsule, length=3, radius=3, center=<7 0 0>, rotation=<rot -90 0 0>}]
 dyn_hair_short_bk_r_02 [{shapeType=Capsule, length=6.6125, radius=3.3875, center=<3.5618 0 0.6656>, rotation=<rot -90 90 0>}]
 dyn_hair_short_bk_r_03 [{shapeType=Capsule, length=5, radius=2.5, center=<4 0 1.5>, rotation=<rot -90 0 0>}]
 dyn_hair_short_bk_r_04 [{shapeType=Capsule, length=3, radius=3.3, center=<3 0 0>, rotation=<rot -90 0 0>}]
 dyn_hair_short_side_l_01 [{shapeType=Capsule, length=3, radius=2, center=<3 0 2>, rotation=<rot -90 90 0>}]
 dyn_hair_short_side_l_02 [{shapeType=Capsule, length=3, radius=2, center=<3 0 2>, rotation=<rot -90 90 0>}]
 dyn_hair_short_side_l_03 [{shapeType=Capsule, length=2.1336, radius=3, center=<2.8422 -1 1>, rotation=<rot -90 90 0>}]
 dyn_hair_short_side_r_01 [{shapeType=Capsule, length=3, radius=2, center=<3 0 2>, rotation=<rot -90 90 0>}]
 dyn_hair_short_side_r_02 [{shapeType=Capsule, length=3, radius=2, center=<3 0 2>, rotation=<rot -90 90 0>}]
 dyn_hair_short_side_r_03 [{shapeType=Capsule, length=3, radius=2.5, center=<3 1 1>, rotation=<rot -90 90 0>}]
 head [{shapeType=Capsule, length=6.1281, radius=5, shapeName=head_capsule, center=<0.2517 1.1618 0>, rotation=<rot -90 0 0>}]
 neck_01 [{shapeType=Capsule, length=5.5335, radius=4.756, shapeName=neck_02_capsule, center=<0.2517 1.1618 0>, rotation=<rot -90 0 0>}]
 neck_02 [{shapeType=Capsule, length=5.2895, radius=5, shapeName=neck_02_capsule, center=<0.2517 1.1618 0>, rotation=<rot -90 0 0>}]
 spine_03 [{shapeType=Capsule, length=7.1034, radius=7.1034, shapeName=spine_03_capsule}]
 spine_04 [{shapeType=Capsule, length=8.0615, radius=8.0615, shapeName=spine_04_capsule}]
 spine_05 [{shapeType=Capsule, length=9.2727, radius=8.2727, shapeName=spine_05_capsule, center=<0 0.2144 0>}]
 upperarm_l [{shapeType=Capsule, length=9, radius=5, shapeName=upperarm_l_capsule, center=<3.3698 -0.0736 -1.4611>, rotation=<rot -90 0 0>}]
 upperarm_r [{shapeType=Capsule, length=9, radius=5.2068, shapeName=upperarm_r_capsule, center=<-3.3698 0.0736 0.7766>, rotation=<rot -90 0 0>}]
}
constraints [
 {
  bone1Name=base_hair_short_bk_c
  bone2Name=head
  swing1LimitDegrees=45
  swing1Motion=Free
  swing2LimitDegrees=45
  swing2Motion=Free
  twistLimitDegrees=45
  twistMotion=Free
 }
 {bone1Name=dyn_hair_short_bk_c_01, bone2Name=base_hair_short_bk_c, swing1LimitDegrees=10, swing2LimitDegrees=45, twistLimitDegrees=5}
 {bone1Name=dyn_hair_short_bk_c_02, bone2Name=dyn_hair_short_bk_c_01, swing1LimitDegrees=15, swing2LimitDegrees=45, twistLimitDegrees=5}
 {
  bone1Name=dyn_hair_short_bk_r_02
  bone2Name=dyn_hair_short_bk_c_02
  swing1LimitDegrees=45
  swing1Motion=Free
  swing2LimitDegrees=45
  swing2Motion=Free
  twistLimitDegrees=45
  twistMotion=Free
 }
 {
  bone1Name=dyn_hair_short_bk_l_02
  bone2Name=dyn_hair_short_bk_c_02
  swing1LimitDegrees=45
  swing1Motion=Free
  swing2LimitDegrees=45
  swing2Motion=Free
  twistLimitDegrees=45
  twistMotion=Free
 }
 {bone1Name=dyn_hair_short_bk_c_03, bone2Name=dyn_hair_short_bk_c_02, swing1LimitDegrees=20, swing2LimitDegrees=45, twistLimitDegrees=5}
 {
  bone1Name=base_hair_short_bk_l
  bone2Name=head
  swing1LimitDegrees=45
  swing1Motion=Free
  swing2LimitDegrees=45
  swing2Motion=Free
  twistLimitDegrees=45
  twistMotion=Free
 }
 {bone1Name=dyn_hair_short_bk_l_01, bone2Name=base_hair_short_bk_l, swing1LimitDegrees=10, swing2LimitDegrees=27.5, twistLimitDegrees=5}
 {bone1Name=dyn_hair_short_bk_l_02, bone2Name=dyn_hair_short_bk_l_01, swing1LimitDegrees=15, swing2LimitDegrees=27.5, twistLimitDegrees=5}
 {bone1Name=dyn_hair_short_bk_l_03, bone2Name=dyn_hair_short_bk_l_02, swing1LimitDegrees=20, swing2LimitDegrees=27.5, twistLimitDegrees=5}
 {
  bone1Name=base_hair_short_bk_r
  bone2Name=head
  swing1LimitDegrees=45
  swing1Motion=Free
  swing2LimitDegrees=45
  swing2Motion=Free
  twistLimitDegrees=45
  twistMotion=Free
 }
 {bone1Name=dyn_hair_short_bk_r_01, bone2Name=base_hair_short_bk_r, swing1LimitDegrees=10, swing2LimitDegrees=45, twistLimitDegrees=5}
 {bone1Name=dyn_hair_short_bk_r_02, bone2Name=dyn_hair_short_bk_r_01, swing1LimitDegrees=15, swing2LimitDegrees=45, twistLimitDegrees=5}
 {
  bone1Name=dyn_hair_short_side_r_03
  bone2Name=dyn_hair_short_bk_r_02
  swing1LimitDegrees=45
  swing1Motion=Free
  swing2LimitDegrees=45
  swing2Motion=Free
  twistLimitDegrees=45
  twistMotion=Free
 }
 {bone1Name=dyn_hair_short_bk_r_03, bone2Name=dyn_hair_short_bk_r_02, swing1LimitDegrees=20, swing2LimitDegrees=45, twistLimitDegrees=5}
 {
  bone1Name=base_hair_short_side_l
  bone2Name=head
  swing1LimitDegrees=45
  swing1Motion=Free
  swing2LimitDegrees=45
  swing2Motion=Free
  twistLimitDegrees=45
  twistMotion=Free
 }
 {bone1Name=dyn_hair_short_side_l_01, bone2Name=base_hair_short_side_l, swing1LimitDegrees=10, swing2LimitDegrees=10, twistLimitDegrees=10}
 {bone1Name=dyn_hair_short_side_l_02, bone2Name=dyn_hair_short_side_l_01, swing1LimitDegrees=10, swing2LimitDegrees=10, twistLimitDegrees=10}
 {bone1Name=dyn_hair_short_side_l_03, bone2Name=dyn_hair_short_side_l_02, swing1LimitDegrees=10, swing2LimitDegrees=10, twistLimitDegrees=10}
 {
  bone1Name=base_hair_short_side_r
  bone2Name=head
  swing1LimitDegrees=45
  swing1Motion=Free
  swing2LimitDegrees=45
  swing2Motion=Free
  twistLimitDegrees=45
  twistMotion=Free
 }
 {bone1Name=dyn_hair_short_side_r_01, bone2Name=base_hair_short_side_r, swing1LimitDegrees=10, swing2LimitDegrees=10, twistLimitDegrees=10}
 {bone1Name=dyn_hair_short_side_r_02, bone2Name=dyn_hair_short_side_r_01, swing1LimitDegrees=10, swing2LimitDegrees=10, twistLimitDegrees=10}
 {bone1Name=dyn_hair_short_side_r_03, bone2Name=dyn_hair_short_side_r_02, swing1LimitDegrees=10, swing2LimitDegrees=10, twistLimitDegrees=10}
 {
  bone1Name=dyn_hair_short_side_l_02
  bone2Name=dyn_hair_short_bk_c_02
  swing1LimitDegrees=45
  swing1Motion=Free
  swing2LimitDegrees=45
  swing2Motion=Free
  twistLimitDegrees=45
  twistMotion=Free
 }
 {
  bone1Name=dyn_hair_short_side_r_02
  bone2Name=dyn_hair_short_bk_c_02
  swing1LimitDegrees=45
  swing1Motion=Free
  swing2LimitDegrees=45
  swing2Motion=Free
  twistLimitDegrees=45
  twistMotion=Free
 }
 {
  bone1Name=dyn_hair_short_side_l_02
  bone2Name=dyn_hair_short_bk_c_02
  swing1LimitDegrees=45
  swing1Motion=Free
  swing2LimitDegrees=45
  swing2Motion=Free
  twistLimitDegrees=45
  twistMotion=Free
 }
 {bone1Name=dyn_hair_short_bk_l_04, bone2Name=dyn_hair_short_bk_l_03, swing1LimitDegrees=25, swing2LimitDegrees=27.5, twistLimitDegrees=5}
 {bone1Name=dyn_hair_short_bk_c_04, bone2Name=dyn_hair_short_bk_c_03, swing1LimitDegrees=25, swing2LimitDegrees=45, twistLimitDegrees=5}
 {bone1Name=dyn_hair_short_bk_r_04, bone2Name=dyn_hair_short_bk_r_03, swing1LimitDegrees=25, swing2LimitDegrees=45, twistLimitDegrees=5}
 {
  bone1Name=dyn_hair_short_bk_l_04
  bone2Name=dyn_hair_short_bk_c_04
  swing1LimitDegrees=45
  swing1Motion=Free
  swing2LimitDegrees=45
  swing2Motion=Free
  twistLimitDegrees=45
  twistMotion=Free
 }
]
